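# Idle ink effect for the Dusty Bun character body that draws four black sketch marks attached to its skeletal mesh.
# A single SketchMarks emitter bursts four particles, and each particle's partition index selects its bone or socket, position offset, sprite size and tilt.
# Particles re-sample the skeletal mesh on update and align to the sampled rotation plus that tilt, with sprite and mesh renderers.
# The front-end preview alpha switch currently resolves to full alpha either way.
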NiagaraSystem /BRCosmetics/Characters/Player/Male/Large/Bodies/M_LRG_DustyBun/FX/NS_Character_DustyBun_Ink_Idle
dependencies {
 usedDataInterfaces [
  &NiagaraDataInterfaceArrayFloat2
  &NiagaraDataInterfaceArrayFloat3
  &NiagaraDataInterfaceArrayInt32
  &NiagaraDataInterfaceSkeletalMesh
 ]
 usedDynamicInputs [&Multiply_Vector]
 usedModules [
  &self:SystemSpawnScript.SystemScriptSource.SystemScriptGraph.NiagaraNodeAssignment_0.SetVariables_#1
  &SystemState
 ]
}
emitters {SketchMarks {
  emitterUpdateScript [
   &EmitterState(
    Inactive Response=::NewEnumerator1
    Life Cycle Mode=::NewEnumerator1
    Loop Behavior=::NewEnumerator1
    Loop Duration Mode=::NewEnumerator1
    UseLoopDelay=false
   )
   &SpawnBurst_Instantaneous(Spawn Count=4, Use Spawn Probability=false, UseLoopCountLimit=false)
  ]
  particleSpawnScript [
   &PartitionParticles(Number Of Partitions=4)
   &InitializeParticle(
    Color=<rgba 0 0 0 1>
    Mesh Scale Mode=::NewEnumerator3
    Sprite Rotation Mode=::NewEnumerator0
    Sprite Size=&SelectVector2DFromArray(
     Direct Array Index=$Particles.PartitionParticles.Partition
     Vector 2D Selection Array=DataInterface() {propertyValues {FloatData [<22 60>, <22 60>, <22 45>, <22 45>]}}
    )
    Sprite Size Mode=::NewEnumerator4
    UsePositionOffset=false
    Write Material Random=true
   )
   &CameraOffset(Camera Offset Amount=2)
  ]
  particleUpdateScript [
   &ParticleState(Kill Particles When Lifetime Has Elapsed=false, Let Infinitely Lived Particles Die When Emitter Deactivates=true)
   &SkeletalMeshLocation(
    Bone / Socket Index=&SelectIntFromArray(Direct Array Index=$Particles.PartitionParticles.Partition, Int Selection Array=DataInterface() {propertyValues {IntData [0,1,2,3]}})
    MaskSpawnGroup=false
    Mesh Sampling Type=::NewEnumerator5
    Sampled Position Offset=&SelectVectorFromArray(
     Direct Array Index=$Particles.PartitionParticles.Partition
     Vector Selection Array=DataInterface() {propertyValues {FloatData [<0 0 3.5>, <-2.8 0 3>, <3.4 0 -2.8>, <-4.6 0 -3>]}}
    )
    Skeletal Mesh=$System.Skeletal Mesh
    Socket Sampling Mode=::NewEnumerator5
   )
   &ScaleColor(
    Scale Alpha=&MakeCustomFloatFromBool(Bool=$User.bIsFrontEndPreview, True Float=1, False Float=1)
    Scale RGB=&VectorFromFloat(Value=1)
    ScaleA=true
    ScaleRGB=true
   )
   &SpriteFacingAndAlignment(
    Alignment Coordinate Space=::Local
    Facing Coordinate Space=::Local
    Sprite Alignment=&ReturnMeshOrientationAxis(
     Mesh Orientation Axis=::XAxis
     Mesh Orientation Quat=&MultiplyQuaternion(
      Quaternion A=$Particles.SkeletalMeshLocation.SampledRotation
      Quaternion B=&MakeQuaternion(
       Quaternion From=0
       Axis=<0 1 0>
       Angle=&MergeHelper_FloatSwitch(0=20, 1=-20, 2=-40, 3=40, Index=$Particles.PartitionParticles.Partition, Num=::NewEnumerator2)
       From=?
       To=?
       Angle Type=::NewEnumerator0
       Normalize Vectors=false
      )
     )
    )
    Sprite Facing=<0 1 0>
    WriteAlignment=true
    WriteFacing=true
   )
  ]
  rendererClasses [&NiagaraSpriteRendererProperties, &NiagaraMeshRendererProperties]
  renderers [
   {rendererClass=&NiagaraSpriteRendererProperties}
   {rendererClass=&NiagaraMeshRendererProperties, rendererIndex=1}
  ]
 }}
summary {
 systemName=NS_Character_DustyBun_Ink_Idle
 emitters [{emitterName=SketchMarks, rendererClasses [&NiagaraSpriteRendererProperties, &NiagaraMeshRendererProperties]}]
 userVariables [
  User.bIsFrontEnd {type=NiagaraBool}
  User.bIsFrontEndPreview {type=NiagaraBool}
  User.bisHandHidden {type=NiagaraBool}
 ]
}
