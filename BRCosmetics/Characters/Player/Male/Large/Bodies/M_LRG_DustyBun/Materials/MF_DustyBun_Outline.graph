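# DustyBun character material function that turns MF_SoftGlowOutlines into a reactive graffiti-style outline, reusing LoneDice parameters and textures.
# It overrides opacity, emissive and world position offset, then applies the base character opacity function.
# Quantized reactivity and time-stepped noise feed normal displacement with camera offset, dithered Fresnel opacity, and reactive color broken up by screen-stabilized spray grunge.
# An IsBackpack switch varies camera offset and two-sided masking; a custom scene-depth node builds a silhouette mask.
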MaterialFunction /BRCosmetics/Characters/Player/Male/Large/Bodies/M_LRG_DustyBun/Materials/MF_DustyBun_Outline
graph {
 properties {previewBlendMode=BLEND_Translucent}
 nodes [
  Constant() {r=1}
  n1 = GetMaterialAttributes(None=MaterialFunctionCall() {materialFunction=/Game/Global/Materials/MaterialGym/Materials/SoftGlowOutlines/MF_SoftGlowOutlines}.Result) {attributeGetTypes [#1, #2, #3]}
  n2 = NamedRerouteDeclaration(None=Constant()) {name=Opacity_IN, nodeColor=<rgba 0.9219 0 1 1>}
  n3 = NamedRerouteUsage()
  n4 = NamedRerouteUsage()
  FunctionOutput(None=MaterialFunctionCall!(Material=SetMaterialAttributes!(MaterialAttributes=@n1.MaterialAttributes, Opacity=@n2, EmissiveColor=@n3, WorldPositionOffset=@n4) {attributeSetTypes [#1, #2, #3]}) {materialFunction=/Game/Global/Materials/Character_Parent_2023/Functions/BaseFN_CharacterOpacity_2023}.Result) {outputName=MaterialAttributes}
  n5 = ScalarParameter() {defaultValue=8, group=LoneDice Graffiti Outline, parameterName=LD_ReactiveSteps}
  n6 = Multiply(
   A=Multiply(
    A=ScalarParameter() {group=LoneDice Graffiti Outline, parameterName=LD_Reactivity, sliderMax=1}
    B=ScalarParameter() {defaultValue=1, group=LoneDice Graffiti Outline, parameterName=LD_ReactivityToggle, sliderMax=1}
   )
   B=@n5
  ) {constB=0.5}
  NamedRerouteDeclaration(None=Divide(A=Round(None=@n6), B=@n5)) {name=Reactivity, nodeColor=<rgba 0 1 0.5078 1>}
  n7 = VertexNormalWS()
  n8 = NamedRerouteUsage()
  n9 = ScalarParameter() {defaultValue=2, group=LoneDice Graffiti Outline, parameterName=LD_Cutout_WPONoiseScale}
  n10 = NamedRerouteUsage()
  n11 = Add(A=NamedRerouteUsage(), B=Multiply(A=Multiply(A=@n7, B=@n8), B=LinearInterpolate(B=@n9, Alpha=@n10)))
  n12 = ScalarParameter() {defaultValue=-15, group=LoneDice Graffiti Outline, parameterName=LD_Cutout_CamOffset}
  n13 = ScalarParameter() {group=LoneDice Graffiti Outline, parameterName=LD_Cutout_CamOffset2}
  n14 = NamedRerouteUsage()
  n15 = Saturate(None=SmoothStep(Value=ComponentMask(None=@n14) {g=true}) {constMax=0, constMin=-10})
  n16 = MaterialFunctionCall(Offset Amount=StaticSwitchParameter(True=LinearInterpolate(A=@n12, B=@n13, Alpha=@n15), False=@n12) {group=LoneDice Graffiti Outline, parameterName=IsBackpack}) {materialFunction=/Engine/Functions/Engine_MaterialFunctions02/WorldPositionOffset/CameraOffset}
  NamedRerouteDeclaration(None=LinearInterpolate(A=NamedRerouteUsage(), B=Add(A=@n11, B=@n16.Clamped Camera Offset), Alpha=NamedRerouteUsage())) {name=WPO_OUT, nodeColor=<rgba 0 0.0859 1 1>}
  n17 = LocalPosition() {localOrigin=InstancePreSkinning}
  NamedRerouteDeclaration(None=ShaderStageSwitch(PixelShader=VertexInterpolator(VS=@n17.XYZ).PS, VertexShader=@n17.XYZ)) {name=PS-LocalPos, nodeColor=<rgba 1 0 0.2578 1>}
  n18 = Fresnel() {baseReflectFraction=0, exponent=1}
  n19 = Multiply(A=SmoothStep(Value=OneMinus(None=@n18)) {constMax=0.5}, B=Constant() {r=2})
  n20 = Multiply(A=NamedRerouteUsage()) {constB=2}
  n21 = MaterialFunctionCall() {materialFunction=&BoundingBoxBased_0-1_UVW}
  n22 = TextureSample(UVs=Panner(Coordinate=ComponentMask(None=@n21.RGB) {b=true, r=true}) {speedY=0.1}) {
   samplerSource=SSM_Wrap_WorldGroupSettings
   texture=/BRCosmetics/Characters/Player/Male/Medium/Bodies/M_MED_LoneDice/Textures/T_LoneDice_Outline_Noise
  }
  n23 = LinearInterpolate(
   A=NamedRerouteUsage()
   B=Saturate(None=MaterialFunctionCall!(Alpha Threshold=@n19, Random) {materialFunction=/Game/Environments/Asteria/Sets/NeonCity/Techart/Screen/MaterialFunctions/DitherTemporalAABlue}.Result)
   Alpha=SmoothStep(Value=Subtract(A=@n20, B=@n22.G)) {constMin=0.9, desc=on reactive draw outline}
  ) {constAlpha=1}
  n24 = Saturate(None=@n23)
  n25 = NamedRerouteUsage()
  NamedRerouteDeclaration(None=Saturate(None=Subtract(A=@n24, B=@n25))) {name=dEBUG, nodeColor=<rgba 1 0 0.9141 1>}
  n26 = NamedRerouteUsage()
  n27 = VectorParameter() {group=LoneDice Graffiti Outline, parameterName=LD_ReactiveColor, defaultValue=<rgba 0.91 0.1393 0.0819 1>}
  n28 = TextureSample(UVs=MaterialFunctionCall!(Position, UVRadialScale=ScalarParameter() {defaultValue=1, group=LoneDice Graffiti Outline, parameterName=LD_GrungeMap_Scale}) {materialFunction=/Game/Characters/Player/Common/MaterialFunctions/Superlevel/CharacterStabilizedScreenSpaceUVs}.StabilizedUVs) {
   samplerType=SAMPLERTYPE_Masks
   texture=/BRCosmetics/Characters/Player/Male/Medium/Bodies/M_MED_LoneDice/Textures/T_LoneDice_Outline_SprayAtlas
  }
  n29 = Subtract(
   A=Multiply(A=@n27.RGB, B=ScalarParameter() {defaultValue=1, group=LoneDice Graffiti Outline, parameterName=LD_ReactiveColor_Strength})
   B=Multiply(A=@n28.G, B=ScalarParameter() {defaultValue=0.15, group=LoneDice Graffiti Outline, parameterName=LD_GrungeMap_Intensity})
  )
  n30 = NamedRerouteUsage()
  NamedRerouteDeclaration(None=MaterialFunctionCall(Emissive=LinearInterpolate(A=@n26, B=@n29, Alpha=@n30)) {materialFunction=&MF_FX_EmissiveAdjustments}.Result) {name=Out_Emissive, nodeColor=<rgba 0 0.7187 1 1>}
  NamedRerouteDeclaration(None=Multiply(A=VertexNormalWS(), B=ScalarParameter() {defaultValue=0.6, group=SoftGlow, parameterName=LD_WPO_Amount})) {name=WPO_IN, nodeColor=<rgba 0.4062 1 0 1>}
  n31 = Divide(A=NamedRerouteUsage()) {constB=64}
  n32 = ScalarParameter() {defaultValue=5, group=LoneDice Graffiti Outline, parameterName=LD_Cutout_Steps}
  n33 = Multiply(A=Multiply(A=Time(), B=ScalarParameter() {defaultValue=0.75, group=LoneDice Graffiti Outline, parameterName=LD_Cutout_Speed}), B=@n32) {constB=0.5}
  n34 = Frac(None=Divide(A=Round(None=@n33), B=@n32))
  n35 = TextureSample(UVs=Add(
    A=ComponentMask(None=@n31) {b=true, g=true}
    B=MaterialFunctionCall(Y=@n34) {materialFunction=&MakeFloat2}.Result
   )) {
   samplerSource=SSM_Wrap_WorldGroupSettings
   texture=&T_LoneDice_Outline_Noise
  }
  NamedRerouteDeclaration(None=@n35.R) {name=Noise, nodeColor=<rgba 0 1 0.9531 1>}
  n36 = OneMinus(None=TwoSidedSign())
  NamedRerouteDeclaration(None=StaticSwitchParameter(True=Multiply(A=@n23, B=@n36), False=@n23) {group=LoneDice Graffiti Outline, parameterName=IsBackpack}) {name=Opacity_OUT, nodeColor=<rgba 0.7578 1 0 1>}
  n37 = NamedRerouteUsage()
  NamedRerouteDeclaration(None=Saturate(None=Ceil(None=@n37))) {name=Reactivity_Ceil, nodeColor=<rgba 1 0.1875 0 1>}
  NamedRerouteDeclaration(None=Constant()) {name=Emissive_IN, nodeColor=<rgba 0.4062 1 0 1>}
  MaterialExpressionMultiply_8 = Multiply(A=Saturate(None=SceneDepth().R), B=PixelDepth().R)
  MaterialExpressionScalarParameter_12 = ScalarParameter() {parameterName=PixelDepthOffset}
  MaterialExpressionScalarParameter_11 = ScalarParameter() {defaultValue=30, parameterName=MaxDepth}
  n38 = Custom!(PDepth=@MaterialExpressionMultiply_8, PixelDepthOffset=@MaterialExpressionScalarParameter_12, DivideF=@MaterialExpressionScalarParameter_11) {
   code="float2 BaseUV = MaterialFloat2(ScreenAlignedPosition(Parameters.ScreenPosition).xy);\r\nfloat tmpdepth = CalcSceneDepth(BaseUV)-PDepth+PixelDepthOffset;\r\nfloat CurColor = (min(1-saturate(max(0,(tmpdepth))/(DivideF+PixelDepthOffset))-0.99,0)+0.99)*saturate(sign(tmpdepth));\r\nreturn CurColor;"
   inputs [
    {
     inputName=PDepth
     input {expression=&self:MaterialExpressionMultiply_8}
    }
    {
     inputName=PixelDepthOffset
     input {expression=&self:MaterialExpressionScalarParameter_12}
    }
    {
     inputName=DivideF
     input {expression=&self:MaterialExpressionScalarParameter_11}
    }
   ]
  }
  n39 = ScalarParameter() {defaultValue=0.01, parameterName=silhouettePower}
  NamedRerouteDeclaration(None=Saturate(None=Power(Base=@n38, Exp=@n39))) {name=SilhouetteMask, nodeColor=<rgba 0 0.8125 1 1>}
  NamedRerouteUsage()
  NamedRerouteUsage()
 ]
}
