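# Material function for the DustyBun character body that turns continuous time into discrete, looping steps for stepped animation in materials.
# Time is multiplied by a scale parameter, wrapped with Frac, multiplied by a step count and floored to a whole step index.
# It outputs that index as FlooredOutput and the index times an offset parameter as TimeStep.
# All three tuning parameters sit in the ToonPBR Blend group.
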MaterialFunction /BRCosmetics/Characters/Player/Male/Large/Bodies/M_LRG_DustyBun/Materials/MF_DustyBunSteppedTime
graph {nodes [
  n1 = Multiply(A=Time(), B=ScalarParameter() {defaultValue=0.2, group=ToonPBR Blend, parameterName=Stepped Time - Scale, sliderMax=1, sortPriority=12})
  n2 = Floor(None=Multiply(
    A=Frac(None=@n1)
    B=ScalarParameter() {defaultValue=12, group=ToonPBR Blend, parameterName=Stepped Time - Steps, sliderMax=32, sortPriority=13}
   ))
  FunctionOutput(None=Multiply(
    A=FunctionOutput(None=@n2) {outputName=FlooredOutput, sortPriority=1}
    B=ScalarParameter() {defaultValue=0.125, group=ToonPBR Blend, parameterName=Stepped Time - Offset, sliderMax=1, sortPriority=14}
   )) {outputName=TimeStep}
 ]}
