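# Material function for the DustyBun character that generates pseudorandom hash values as Float, Vec2 and Vec3 outputs, each remapped to -1..1.
# Each uses a sine-and-frac hash, with vector inputs first dotted against large constants per channel.
# Seed offsets the large multiplier for Vec2 and Vec3 X/Y, but not Float or Vec3 Z, which also lack that multiplier.
# Inputs default to time, quantized UVs and quantized world position.
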MaterialFunction /BRCosmetics/Characters/Player/Male/Large/Bodies/M_LRG_DustyBun/Materials/MF_DustyBun_Hash
graph {
 properties {bExposeToLibrary=true, description=Generate Psuedorandom numbers}
 nodes [
  n1 = FunctionInput(Preview=Time()) {bUsePreviewValueAsDefault=true, inputName=Float, inputType=FunctionInput_Scalar}
  n2 = Frac(None=Multiply(A=Sine(None=@n1)))
  FunctionOutput(None=Subtract(A=Multiply(A=@n2) {constB=2})) {outputName=Float}
  n3 = TextureCoordinate()
  n4 = Constant() {r=10}
  n5 = FunctionInput(Preview=Ceil(None=Multiply(A=@n3, B=@n4))) {bUsePreviewValueAsDefault=true, inputName=Vec2, inputType=FunctionInput_Vector2}
  n6 = Sine(None=DotProduct(A=@n5, B=Constant2Vector() {g=724.3587, r=235.9474}))
  n7 = Add(
   A=Constant() {r=43758.5469}
   B=FunctionInput(Preview=Constant() {r=2}) {bUsePreviewValueAsDefault=true, inputName=Seed, inputType=FunctionInput_Scalar}
  )
  n8 = Multiply(A=Frac(None=Multiply(A=@n6, B=@n7))) {constB=2}
  n9 = Sine(None=DotProduct(A=@n5, B=Constant2Vector() {g=284.3817, r=574.9868}))
  n10 = Multiply(A=Frac(None=Multiply(A=@n9, B=@n7))) {constB=2}
  FunctionOutput(None=AppendVector(A=Subtract(A=@n8), B=Subtract(A=@n10))) {outputName=Vec2}
  n11 = Add(A=WorldPosition().XYZ) {constB=0.5}
  n12 = Constant() {r=32}
  n13 = FunctionInput(Preview=Ceil(None=Divide(A=@n11, B=@n12))) {bUsePreviewValueAsDefault=true, inputName=Vec3}
  n14 = Sine(None=DotProduct(A=@n13, B=Constant3Vector() {constant=<rgba 574.9868 284.3817 847.0239 0>}))
  n15 = Multiply(A=Frac(None=Multiply(A=@n14, B=@n7))) {constB=2}
  n16 = Sine(None=DotProduct(A=@n13, B=Constant3Vector() {constant=<rgba 235.9474 724.3587 485.192 0>}))
  n17 = Multiply(A=Frac(None=Multiply(A=@n16, B=@n7))) {constB=2}
  n18 = Sine(None=DotProduct(A=@n13, B=Constant3Vector() {constant=<rgba 639.9105 178.392 597.2471 0>}))
  n19 = Multiply(A=Frac(None=Multiply(A=@n18))) {constB=2}
  FunctionOutput(None=MaterialFunctionCall(X=Subtract(A=@n15), Y=Subtract(A=@n17), Z=Subtract(A=@n19)) {materialFunction=&MakeFloat3}.Result) {outputName=Vec3}
 ]
}
