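# Material function for the DustyBun character body that layers animated, flickering line strokes over incoming material attributes.
# Time thresholds build one-hot RGB masks that pick stroke frames by dot product, with a second offset clock for normal strokes.
# Line and random-line textures feed paired flipbook branches, while stepped-time hashing of pre-skinned position creates a normal wiggle.
# It overrides base color, specular and normal; the roughness input is left unconnected.
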MaterialFunction /BRCosmetics/Characters/Player/Male/Large/Bodies/M_LRG_DustyBun/Materials/MF_DustyBun_Body
graph {expressions {
  MaterialExpressionAdd_0 {
   class=MaterialExpressionAdd
   inputs [
    A {from=MaterialExpressionTime_1}
    B {from=""}
   ]
   properties {constB=0.05}
  }
  MaterialExpressionAdd_1 {
   class=MaterialExpressionAdd
   inputs [
    A {from=MaterialExpressionNamedRerouteUsage_11}
    B {from=MaterialExpressionMultiply_2}
   ]
  }
  MaterialExpressionConstant_0 {class=MaterialExpressionConstant, properties {r=1}}
  MaterialExpressionConstant_1 {class=MaterialExpressionConstant, properties {r=1}}
  MaterialExpressionConstant_2 {class=MaterialExpressionConstant}
  MaterialExpressionConstant_3 {class=MaterialExpressionConstant, properties {r=1}}
  MaterialExpressionConstant_4 {class=MaterialExpressionConstant}
  MaterialExpressionConstant_5 {class=MaterialExpressionConstant}
  MaterialExpressionConstant_6 {class=MaterialExpressionConstant, properties {r=1}}
  MaterialExpressionConstant_7 {class=MaterialExpressionConstant}
  MaterialExpressionConvert_0 {
   class=MaterialExpressionConvert
   inputs [
    None {from=MaterialExpressionIf_0}
    None {from=MaterialExpressionOneMinus_0}
    None {from=MaterialExpressionIf_3}
   ]
   properties {
    nodeName=Make Float3
    convertInputs [
     {
      type=Scalar
      expressionInput {
       expression=&self:MaterialExpressionIf_0
       inputName=None
       mask=0
       maskA=0
       maskB=0
       maskG=0
       maskR=0
       outputIndex=0
      }
     }
     {
      type=Scalar
      expressionInput {
       expression=&self:MaterialExpressionOneMinus_0
       inputName=None
       mask=0
       maskA=0
       maskB=0
       maskG=0
       maskR=0
       outputIndex=0
      }
     }
     {
      type=Scalar
      expressionInput {
       expression=&self:MaterialExpressionIf_3
       inputName=None
       mask=0
       maskA=0
       maskB=0
       maskG=0
       maskR=0
       outputIndex=0
      }
     }
    ]
    convertOutputs [{type=Vector3}]
   }
  }
  MaterialExpressionConvert_1 {
   class=MaterialExpressionConvert
   inputs [
    None {from=MaterialExpressionIf_4}
    None {from=MaterialExpressionOneMinus_2}
    None {from=MaterialExpressionIf_5}
   ]
   properties {
    nodeName=Make Float3
    convertInputs [
     {
      type=Scalar
      expressionInput {
       expression=&self:MaterialExpressionIf_4
       inputName=None
       mask=0
       maskA=0
       maskB=0
       maskG=0
       maskR=0
       outputIndex=0
      }
     }
     {
      type=Scalar
      expressionInput {
       expression=&self:MaterialExpressionOneMinus_2
       inputName=None
       mask=0
       maskA=0
       maskB=0
       maskG=0
       maskR=0
       outputIndex=0
      }
     }
     {
      type=Scalar
      expressionInput {
       expression=&self:MaterialExpressionIf_5
       inputName=None
       mask=0
       maskA=0
       maskB=0
       maskG=0
       maskR=0
       outputIndex=0
      }
     }
    ]
    convertOutputs [{type=Vector3}]
   }
  }
  MaterialExpressionDivide_0 {
   class=MaterialExpressionDivide
   inputs [
    A {from=MaterialExpressionScalarParameter_16}
    B {from=""}
   ]
   properties {constB=3}
  }
  MaterialExpressionDotProduct_0 {
   class=MaterialExpressionDotProduct
   inputs [
    A {from=MaterialExpressionNamedRerouteUsage_2, output=""}
    B {from=MaterialExpressionConvert_0, output=""}
   ]
   properties {}
  }
  MaterialExpressionDotProduct_1 {
   class=MaterialExpressionDotProduct
   inputs [
    A {from=MaterialExpressionOneMinus_3, output=""}
    B {from=MaterialExpressionConvert_1, output=""}
   ]
   properties {}
  }
  MaterialExpressionFrac_0 {class=MaterialExpressionFrac, inputs [None {from=MaterialExpressionMultiply_24, output=""}], properties {}}
  MaterialExpressionFrac_1 {class=MaterialExpressionFrac, inputs [None {from=MaterialExpressionMultiply_8, output=""}], properties {}}
  MaterialExpressionFunctionInput_0 {
   class=MaterialExpressionFunctionInput
   inputs [Preview {from="", output=""}]
   properties {inputName=In_MaterialAttributes, inputType=FunctionInput_MaterialAttributes}
  }
  MaterialExpressionFunctionOutput_0 {class=MaterialExpressionFunctionOutput, inputs [None {from=MaterialExpressionSetMaterialAttributes_0, output=""}]}
  MaterialExpressionGetMaterialAttributes_0 {
   class=MaterialExpressionGetMaterialAttributes
   inputs [None {from=MaterialExpressionNamedRerouteUsage_3, output=""}]
   properties {attributeGetTypes [
     #1
     #2
     #3
     #4
    ]}
  }
  MaterialExpressionIf_0 {
   class=MaterialExpressionIf
   inputs [
    A {from=MaterialExpressionMultiply_10, output=""}
    B {from="", output=""}
    "A > B" {from=MaterialExpressionConstant_2, output=""}
    "A == B" {from=MaterialExpressionConstant_2, output=""}
    "A < B" {from=MaterialExpressionConstant_1, output=""}
   ]
   properties {constB=0.111}
  }
  MaterialExpressionIf_1 {
   class=MaterialExpressionIf
   inputs [
    A {from=MaterialExpressionNamedRerouteUsage_39, output=""}
    B {from="", output=""}
    "A > B" {from=MaterialExpressionNamedRerouteUsage_44, output=""}
    "A == B" {from=MaterialExpressionNamedRerouteUsage_44, output=""}
    "A < B" {from=MaterialExpressionNamedRerouteUsage_45, output=""}
   ]
   properties {constB=0.5}
  }
  MaterialExpressionIf_2 {
   class=MaterialExpressionIf
   inputs [
    A {from=MaterialExpressionNamedRerouteUsage_46, output=""}
    B {from="", output=""}
    "A > B" {from=MaterialExpressionNamedRerouteUsage_49, output=""}
    "A == B" {from=MaterialExpressionNamedRerouteUsage_49, output=""}
    "A < B" {from=MaterialExpressionNamedRerouteUsage_50, output=""}
   ]
   properties {constB=0.5}
  }
  MaterialExpressionIf_3 {
   class=MaterialExpressionIf
   inputs [
    A {from=MaterialExpressionMultiply_10, output=""}
    B {from="", output=""}
    "A > B" {from=MaterialExpressionConstant_3, output=""}
    "A == B" {from=MaterialExpressionConstant_4, output=""}
    "A < B" {from=MaterialExpressionConstant_4, output=""}
   ]
   properties {constB=0.222}
  }
  MaterialExpressionIf_4 {
   class=MaterialExpressionIf
   inputs [
    A {from=MaterialExpressionAdd_0, output=""}
    B {from="", output=""}
    "A > B" {from=MaterialExpressionConstant_5, output=""}
    "A == B" {from=MaterialExpressionConstant_5, output=""}
    "A < B" {from=MaterialExpressionConstant_0, output=""}
   ]
   properties {constB=0.161}
  }
  MaterialExpressionIf_5 {
   class=MaterialExpressionIf
   inputs [
    A {from=MaterialExpressionAdd_0, output=""}
    B {from="", output=""}
    "A > B" {from=MaterialExpressionConstant_6, output=""}
    "A == B" {from=MaterialExpressionConstant_7, output=""}
    "A < B" {from=MaterialExpressionConstant_7, output=""}
   ]
   properties {constB=0.272}
  }
  MaterialExpressionIf_6 {
   class=MaterialExpressionIf
   inputs [
    A {from=MaterialExpressionNamedRerouteUsage_20, output=""}
    B {from="", output=""}
    "A > B" {from=MaterialExpressionNamedRerouteUsage_27, output=""}
    "A == B" {from=MaterialExpressionNamedRerouteUsage_27, output=""}
    "A < B" {from=MaterialExpressionNamedRerouteUsage_26, output=""}
   ]
   properties {constB=0.5}
  }
  MaterialExpressionIf_7 {
   class=MaterialExpressionIf
   inputs [
    A {from=MaterialExpressionNamedRerouteUsage_23, output=""}
    B {from="", output=""}
    "A > B" {from=MaterialExpressionNamedRerouteUsage_28, output=""}
    "A == B" {from=MaterialExpressionNamedRerouteUsage_28, output=""}
    "A < B" {from=MaterialExpressionNamedRerouteUsage_29, output=""}
   ]
   properties {constB=0.5}
  }
  MaterialExpressionLinearInterpolate_0 {
   class=MaterialExpressionLinearInterpolate
   inputs [
    A {from=""}
    B {from=MaterialExpressionLinearInterpolate_4}
    Alpha {from=MaterialExpressionScalarParameter_4}
   ]
   properties {constA=1}
  }
  MaterialExpressionLinearInterpolate_1 {
   class=MaterialExpressionLinearInterpolate
   inputs [
    A {from=MaterialExpressionOneMinus_1}
    B {from=MaterialExpressionGetMaterialAttributes_0, output=Roughness}
    Alpha {from=MaterialExpressionNamedRerouteUsage_36}
   ]
   properties {constA=1}
  }
  MaterialExpressionLinearInterpolate_3 {
   class=MaterialExpressionLinearInterpolate
   inputs [
    A {from=MaterialExpressionMultiply_1}
    B {from=MaterialExpressionMultiply_16}
    Alpha {from=MaterialExpressionNamedRerouteUsage_43}
   ]
   properties {constB=0}
  }
  MaterialExpressionLinearInterpolate_4 {
   class=MaterialExpressionLinearInterpolate
   inputs [
    A {from=MaterialExpressionNamedRerouteUsage_8}
    B {from=MaterialExpressionNamedRerouteUsage_4}
    Alpha {from=MaterialExpressionNamedRerouteUsage_8}
   ]
  }
  MaterialExpressionLinearInterpolate_6 {
   class=MaterialExpressionLinearInterpolate
   inputs [
    A {from=MaterialExpressionNamedRerouteUsage_36}
    B {from=MaterialExpressionGetMaterialAttributes_0, output=Specular}
    Alpha {from=MaterialExpressionReroute_1}
   ]
  }
  MaterialExpressionLocalPosition_1 {class=MaterialExpressionLocalPosition, inputs [], properties {localOrigin=InstancePreSkinning}}
  MaterialExpressionMaterialFunctionCall_1 {
   class=MaterialExpressionMaterialFunctionCall
   inputs [
    "Animation  Phase (0-1)" {from=MaterialExpressionNamedRerouteUsage_18}
    Number of Rows {from=MaterialExpressionNamedRerouteDeclaration_12}
    Number of Columns {from=MaterialExpressionNamedRerouteDeclaration_13}
    Texture {from=MaterialExpressionNamedRerouteDeclaration_18}
    UVs {from=""}
    MipBias/Level {from=""}
    "Use Mip Bias (T) Level (F)" {from=""}
    "Clamp Anim (see tooltip)" {from=""}
   ]
   properties {materialFunction=&FlipBook}
  }
  MaterialExpressionMaterialFunctionCall_14 {
   class=MaterialExpressionMaterialFunctionCall
   inputs [
    "Animation  Phase (0-1)" {from=MaterialExpressionNamedRerouteUsage_0}
    Number of Rows {from=MaterialExpressionNamedRerouteDeclaration_20}
    Number of Columns {from=MaterialExpressionNamedRerouteDeclaration_21}
    Texture {from=MaterialExpressionNamedRerouteDeclaration_30}
    UVs {from=""}
    MipBias/Level {from=""}
    "Use Mip Bias (T) Level (F)" {from=""}
    "Clamp Anim (see tooltip)" {from=""}
   ]
   properties {materialFunction=&FlipBook}
  }
  MaterialExpressionMaterialFunctionCall_2 {
   class=MaterialExpressionMaterialFunctionCall
   inputs []
   properties {materialFunction=&/MF_DustyBunSteppedTime}
  }
  MaterialExpressionMaterialFunctionCall_3 {
   class=MaterialExpressionMaterialFunctionCall
   inputs [
    Seed {from=MaterialExpressionMultiply_7}
    Float {from=""}
    Vec3 {from=""}
    Vec2 {from=""}
   ]
   properties {materialFunction=&/MF_DustyBun_Hash}
  }
  MaterialExpressionMaterialFunctionCall_4 {
   class=MaterialExpressionMaterialFunctionCall
   inputs [
    "Animation  Phase (0-1)" {from=MaterialExpressionNamedRerouteUsage_19}
    Number of Rows {from=MaterialExpressionNamedRerouteDeclaration_12}
    Number of Columns {from=MaterialExpressionNamedRerouteDeclaration_13}
    Texture {from=MaterialExpressionNamedRerouteDeclaration_23}
    UVs {from=""}
    MipBias/Level {from=""}
    "Use Mip Bias (T) Level (F)" {from=""}
    "Clamp Anim (see tooltip)" {from=""}
   ]
   properties {materialFunction=&FlipBook}
  }
  MaterialExpressionMaterialFunctionCall_8 {
   class=MaterialExpressionMaterialFunctionCall
   inputs [
    "Animation  Phase (0-1)" {from=MaterialExpressionNamedRerouteUsage_1}
    Number of Rows {from=MaterialExpressionNamedRerouteDeclaration_20}
    Number of Columns {from=MaterialExpressionNamedRerouteDeclaration_21}
    Texture {from=MaterialExpressionNamedRerouteDeclaration_31}
    UVs {from=""}
    MipBias/Level {from=""}
    "Use Mip Bias (T) Level (F)" {from=""}
    "Clamp Anim (see tooltip)" {from=""}
   ]
   properties {materialFunction=&FlipBook}
  }
  MaterialExpressionMax_0 {
   class=MaterialExpressionMax
   inputs [
    A {from=MaterialExpressionIf_4}
    B {from=MaterialExpressionIf_5}
   ]
  }
  MaterialExpressionMax_1 {
   class=MaterialExpressionMax
   inputs [
    A {from=MaterialExpressionIf_0}
    B {from=MaterialExpressionIf_3}
   ]
  }
  MaterialExpressionMultiply_0 {
   class=MaterialExpressionMultiply
   inputs [
    A {from=MaterialExpressionNamedRerouteUsage_5}
    B {from=MaterialExpressionLinearInterpolate_0}
   ]
  }
  MaterialExpressionMultiply_1 {
   class=MaterialExpressionMultiply
   inputs [
    A {from=MaterialExpressionNamedRerouteUsage_7}
    B {from=MaterialExpressionNamedRerouteUsage_10}
   ]
  }
  MaterialExpressionMultiply_10 {
   class=MaterialExpressionMultiply
   inputs [
    A {from=""}
    B {from=MaterialExpressionFrac_1}
   ]
   properties {constA=0.33}
  }
  MaterialExpressionMultiply_13 {
   class=MaterialExpressionMultiply
   inputs [
    A {from=MaterialExpressionDotProduct_0}
    B {from=""}
   ]
  }
  MaterialExpressionMultiply_16 {
   class=MaterialExpressionMultiply
   inputs [
    A {from=MaterialExpressionNamedRerouteUsage_6}
    B {from=MaterialExpressionMultiply_17}
   ]
  }
  MaterialExpressionMultiply_17 {
   class=MaterialExpressionMultiply
   inputs [
    A {from=MaterialExpressionNamedRerouteUsage_32}
    B {from=""}
   ]
  }
  MaterialExpressionMultiply_2 {
   class=MaterialExpressionMultiply
   inputs [
    A {from=MaterialExpressionSubtract_2}
    B {from=MaterialExpressionNamedRerouteUsage_47}
   ]
  }
  MaterialExpressionMultiply_24 {
   class=MaterialExpressionMultiply
   inputs [
    A {from=MaterialExpressionTime_8}
    B {from=MaterialExpressionNamedRerouteDeclaration_22}
   ]
   properties {constB=0.5}
  }
  MaterialExpressionMultiply_3 {
   class=MaterialExpressionMultiply
   inputs [
    A {from=MaterialExpressionTransform_0}
    B {from=MaterialExpressionMultiply_5}
   ]
  }
  MaterialExpressionMultiply_4 {
   class=MaterialExpressionMultiply
   inputs [
    A {from=MaterialExpressionScalarParameter_1}
    B {from=""}
   ]
   properties {constB=0.03}
  }
  MaterialExpressionMultiply_5 {
   class=MaterialExpressionMultiply
   inputs [
    A {from=MaterialExpressionMaterialFunctionCall_3, output=Vec3}
    B {from=MaterialExpressionMultiply_4}
   ]
  }
  MaterialExpressionMultiply_6 {
   class=MaterialExpressionMultiply
   inputs [
    A {from=MaterialExpressionMaterialFunctionCall_2, output=TimeStep}
    B {from=""}
   ]
   properties {constB=8}
  }
  MaterialExpressionMultiply_7 {
   class=MaterialExpressionMultiply
   inputs [
    A {from=MaterialExpressionMultiply_6}
    B {from=MaterialExpressionMultiply_9}
   ]
  }
  MaterialExpressionMultiply_8 {
   class=MaterialExpressionMultiply
   inputs [
    A {from=MaterialExpressionTime_2}
    B {from=MaterialExpressionScalarParameter_0}
   ]
  }
  MaterialExpressionMultiply_9 {
   class=MaterialExpressionMultiply
   inputs [
    A {from=MaterialExpressionVertexInterpolator_0, output=PS}
    B {from=""}
   ]
  }
  MaterialExpressionNamedRerouteDeclaration_0 {
   class=MaterialExpressionNamedRerouteDeclaration
   inputs [None {from=MaterialExpressionFunctionInput_0}]
   properties {name=In_MaterialAttributes, nodeColor=<rgba 0 0.5547 1 1>}
  }
  MaterialExpressionNamedRerouteDeclaration_1 {
   class=MaterialExpressionNamedRerouteDeclaration
   inputs [None {from=MaterialExpressionMultiply_13}]
   properties {name=Strokes, nodeColor=<rgba 0.0759 0.1154 0.7031 1>}
  }
  MaterialExpressionNamedRerouteDeclaration_10 {
   class=MaterialExpressionNamedRerouteDeclaration
   inputs [None {from=MaterialExpressionTextureSample_2, output=RGB}]
   properties {name=FlipbookResult_1, nodeColor=<rgba 0.4297 1 0 1>}
  }
  MaterialExpressionNamedRerouteDeclaration_11 {
   class=MaterialExpressionNamedRerouteDeclaration
   inputs [None {from=MaterialExpressionTextureCoordinate_1}]
   properties {name=FlipbookUVs_1, nodeColor=<rgba 0 1 0.4609 1>}
  }
  MaterialExpressionNamedRerouteDeclaration_12 {
   class=MaterialExpressionNamedRerouteDeclaration
   inputs [None {from=MaterialExpressionScalarParameter_3}]
   properties {name=RowSize_1, nodeColor=<rgba 0.0781 0 1 1>}
  }
  MaterialExpressionNamedRerouteDeclaration_13 {
   class=MaterialExpressionNamedRerouteDeclaration
   inputs [None {from=MaterialExpressionScalarParameter_2}]
   properties {name=ColumnSize_1, nodeColor=<rgba 0.2422 1 0 1>}
  }
  MaterialExpressionNamedRerouteDeclaration_14 {
   class=MaterialExpressionNamedRerouteDeclaration
   inputs [None {from=MaterialExpressionTextureSample_2, output=RGB}]
   properties {name=FlipbookResult_3, nodeColor=<rgba 0 0.9062 1 1>}
  }
  MaterialExpressionNamedRerouteDeclaration_15 {
   class=MaterialExpressionNamedRerouteDeclaration
   inputs [None {from=MaterialExpressionNamedRerouteDeclaration_11}]
   properties {name=FlipbookUVs_3, nodeColor=<rgba 1 0.7734 0 1>}
  }
  MaterialExpressionNamedRerouteDeclaration_16 {
   class=MaterialExpressionNamedRerouteDeclaration
   inputs [None {from=MaterialExpressionIf_1}]
   properties {name=Flipbook, nodeColor=<rgba 0.6641 1 0 1>}
  }
  MaterialExpressionNamedRerouteDeclaration_17 {
   class=MaterialExpressionNamedRerouteDeclaration
   inputs [None {from=MaterialExpressionTextureSample_0, output=RGB}]
   properties {name=FlipbookResult, nodeColor=<rgba 0.4297 1 0 1>}
  }
  MaterialExpressionNamedRerouteDeclaration_18 {
   class=MaterialExpressionNamedRerouteDeclaration
   inputs [None {from=MaterialExpressionTextureObjectParameter_0}]
   properties {name=FirstFlipbook_1, nodeColor=<rgba 0 1 0.3672 1>}
  }
  MaterialExpressionNamedRerouteDeclaration_19 {
   class=MaterialExpressionNamedRerouteDeclaration
   inputs [None {from=MaterialExpressionTextureCoordinate_0}]
   properties {name=FlipbookUVs, nodeColor=<rgba 0 1 0.4609 1>}
  }
  MaterialExpressionNamedRerouteDeclaration_2 {
   class=MaterialExpressionNamedRerouteDeclaration
   inputs [None {from=MaterialExpressionDotProduct_1}]
   properties {name=NormalOffsetsFromStrokes, nodeColor=<rgba 0.7001 0.7031 0.1204 1>}
  }
  MaterialExpressionNamedRerouteDeclaration_20 {
   class=MaterialExpressionNamedRerouteDeclaration
   inputs [None {from=MaterialExpressionScalarParameter_15}]
   properties {name=RowSize, nodeColor=<rgba 0.0781 0 1 1>}
  }
  MaterialExpressionNamedRerouteDeclaration_21 {
   class=MaterialExpressionNamedRerouteDeclaration
   inputs [None {from=MaterialExpressionScalarParameter_8}]
   properties {name=ColumnSize, nodeColor=<rgba 0.2422 1 0 1>}
  }
  MaterialExpressionNamedRerouteDeclaration_22 {
   class=MaterialExpressionNamedRerouteDeclaration
   inputs [None {from=MaterialExpressionScalarParameter_16}]
   properties {name=FlipbookTimeMultiplier, nodeColor=<rgba 1 0.8672 0 1>}
  }
  MaterialExpressionNamedRerouteDeclaration_23 {
   class=MaterialExpressionNamedRerouteDeclaration
   inputs [None {from=MaterialExpressionTextureObjectParameter_3}]
   properties {name=SecondFlipbook_1, nodeColor=<rgba 1 0.0469 0 1>}
  }
  MaterialExpressionNamedRerouteDeclaration_24 {
   class=MaterialExpressionNamedRerouteDeclaration
   inputs [None {from=MaterialExpressionFrac_0}]
   properties {name=NormalizedTime, nodeColor=<rgba 1 0 0.5859 1>}
  }
  MaterialExpressionNamedRerouteDeclaration_25 {
   class=MaterialExpressionNamedRerouteDeclaration
   inputs [None {from=MaterialExpressionTextureSample_1, output=RGB}]
   properties {name=FlipbookResult_2, nodeColor=<rgba 0 0.9062 1 1>}
  }
  MaterialExpressionNamedRerouteDeclaration_26 {
   class=MaterialExpressionNamedRerouteDeclaration
   inputs [None {from=MaterialExpressionNamedRerouteDeclaration_19}]
   properties {name=FlipbookUVs_2, nodeColor=<rgba 1 0.7734 0 1>}
  }
  MaterialExpressionNamedRerouteDeclaration_27 {
   class=MaterialExpressionNamedRerouteDeclaration
   inputs [None {from=MaterialExpressionIf_6}]
   properties {name=Flipbook_1, nodeColor=<rgba 0.6641 1 0 1>}
  }
  MaterialExpressionNamedRerouteDeclaration_28 {
   class=MaterialExpressionNamedRerouteDeclaration
   inputs [None {from=MaterialExpressionIf_2}]
   properties {name=CurrentFlipbookUV, nodeColor=<rgba 0 1 0.1328 1>}
  }
  MaterialExpressionNamedRerouteDeclaration_29 {
   class=MaterialExpressionNamedRerouteDeclaration
   inputs [None {from=MaterialExpressionIf_7}]
   properties {name=CurrentFlipbookUV_1, nodeColor=<rgba 0 1 0.1328 1>}
  }
  MaterialExpressionNamedRerouteDeclaration_3 {
   class=MaterialExpressionNamedRerouteDeclaration
   inputs [None {from=MaterialExpressionGetMaterialAttributes_0, output=BaseColor}]
   properties {name=In_BaseColor, nodeColor=<rgba 0.6641 1 0 1>}
  }
  MaterialExpressionNamedRerouteDeclaration_30 {
   class=MaterialExpressionNamedRerouteDeclaration
   inputs [None {from=MaterialExpressionTextureObjectParameter_1}]
   properties {name=FirstFlipbook, nodeColor=<rgba 0 1 0.3672 1>}
  }
  MaterialExpressionNamedRerouteDeclaration_31 {
   class=MaterialExpressionNamedRerouteDeclaration
   inputs [None {from=MaterialExpressionTextureObjectParameter_2}]
   properties {name=SecondFlipbook, nodeColor=<rgba 1 0.0469 0 1>}
  }
  MaterialExpressionNamedRerouteDeclaration_4 {
   class=MaterialExpressionNamedRerouteDeclaration
   inputs [None {from=MaterialExpressionGetMaterialAttributes_0, output=Normal}]
   properties {name=In_Normal, nodeColor=<rgba 0 0.1094 1 1>}
  }
  MaterialExpressionNamedRerouteDeclaration_5 {
   class=MaterialExpressionNamedRerouteDeclaration
   inputs [None {from=MaterialExpressionNormalize_1}]
   properties {name=Out_Normal, nodeColor=<rgba 0.9219 1 0 1>}
  }
  MaterialExpressionNamedRerouteDeclaration_8 {
   class=MaterialExpressionNamedRerouteDeclaration
   inputs [None {from=MaterialExpressionMultiply_3}]
   properties {name=VertexNormalWiggle, nodeColor=<rgba 1 0 0.2344 1>}
  }
  MaterialExpressionNamedRerouteDeclaration_9 {
   class=MaterialExpressionNamedRerouteDeclaration
   inputs [None {from=MaterialExpressionMultiply_0}]
   properties {name=Out_BaseColor, nodeColor=<rgba 0.9224 0.1228 1 1>}
  }
  MaterialExpressionNamedRerouteUsage_0 {class=MaterialExpressionNamedRerouteUsage}
  MaterialExpressionNamedRerouteUsage_1 {class=MaterialExpressionNamedRerouteUsage}
  MaterialExpressionNamedRerouteUsage_10 {class=MaterialExpressionNamedRerouteUsage}
  MaterialExpressionNamedRerouteUsage_11 {class=MaterialExpressionNamedRerouteUsage}
  MaterialExpressionNamedRerouteUsage_12 {class=MaterialExpressionNamedRerouteUsage}
  MaterialExpressionNamedRerouteUsage_13 {class=MaterialExpressionNamedRerouteUsage}
  MaterialExpressionNamedRerouteUsage_14 {class=MaterialExpressionNamedRerouteUsage}
  MaterialExpressionNamedRerouteUsage_17 {class=MaterialExpressionNamedRerouteUsage}
  MaterialExpressionNamedRerouteUsage_18 {class=MaterialExpressionNamedRerouteUsage}
  MaterialExpressionNamedRerouteUsage_19 {class=MaterialExpressionNamedRerouteUsage}
  MaterialExpressionNamedRerouteUsage_2 {class=MaterialExpressionNamedRerouteUsage}
  MaterialExpressionNamedRerouteUsage_20 {class=MaterialExpressionNamedRerouteUsage}
  MaterialExpressionNamedRerouteUsage_23 {class=MaterialExpressionNamedRerouteUsage}
  MaterialExpressionNamedRerouteUsage_26 {class=MaterialExpressionNamedRerouteUsage}
  MaterialExpressionNamedRerouteUsage_27 {class=MaterialExpressionNamedRerouteUsage}
  MaterialExpressionNamedRerouteUsage_28 {class=MaterialExpressionNamedRerouteUsage}
  MaterialExpressionNamedRerouteUsage_29 {class=MaterialExpressionNamedRerouteUsage}
  MaterialExpressionNamedRerouteUsage_3 {class=MaterialExpressionNamedRerouteUsage}
  MaterialExpressionNamedRerouteUsage_30 {class=MaterialExpressionNamedRerouteUsage}
  MaterialExpressionNamedRerouteUsage_32 {class=MaterialExpressionNamedRerouteUsage}
  MaterialExpressionNamedRerouteUsage_36 {class=MaterialExpressionNamedRerouteUsage}
  MaterialExpressionNamedRerouteUsage_39 {class=MaterialExpressionNamedRerouteUsage}
  MaterialExpressionNamedRerouteUsage_4 {class=MaterialExpressionNamedRerouteUsage}
  MaterialExpressionNamedRerouteUsage_43 {class=MaterialExpressionNamedRerouteUsage}
  MaterialExpressionNamedRerouteUsage_44 {class=MaterialExpressionNamedRerouteUsage}
  MaterialExpressionNamedRerouteUsage_45 {class=MaterialExpressionNamedRerouteUsage}
  MaterialExpressionNamedRerouteUsage_46 {class=MaterialExpressionNamedRerouteUsage}
  MaterialExpressionNamedRerouteUsage_47 {class=MaterialExpressionNamedRerouteUsage}
  MaterialExpressionNamedRerouteUsage_49 {class=MaterialExpressionNamedRerouteUsage}
  MaterialExpressionNamedRerouteUsage_5 {class=MaterialExpressionNamedRerouteUsage}
  MaterialExpressionNamedRerouteUsage_50 {class=MaterialExpressionNamedRerouteUsage}
  MaterialExpressionNamedRerouteUsage_6 {class=MaterialExpressionNamedRerouteUsage}
  MaterialExpressionNamedRerouteUsage_7 {class=MaterialExpressionNamedRerouteUsage}
  MaterialExpressionNamedRerouteUsage_8 {class=MaterialExpressionNamedRerouteUsage}
  MaterialExpressionNamedRerouteUsage_9 {class=MaterialExpressionNamedRerouteUsage}
  MaterialExpressionNormalize_1 {class=MaterialExpressionNormalize, inputs [VectorInput {from=MaterialExpressionAdd_1, output=""}]}
  MaterialExpressionOneMinus_0 {class=MaterialExpressionOneMinus, inputs [None {from=MaterialExpressionMax_1}]}
  MaterialExpressionOneMinus_1 {class=MaterialExpressionOneMinus, inputs [None {from=MaterialExpressionNamedRerouteUsage_9}]}
  MaterialExpressionOneMinus_2 {class=MaterialExpressionOneMinus, inputs [None {from=MaterialExpressionMax_0}]}
  MaterialExpressionOneMinus_3 {class=MaterialExpressionOneMinus, inputs [None {from=MaterialExpressionNamedRerouteUsage_30}]}
  MaterialExpressionReroute_0 {class=MaterialExpressionReroute, inputs [None {from=MaterialExpressionGetMaterialAttributes_0, output=MaterialAttributes}]}
  MaterialExpressionReroute_1 {class=MaterialExpressionReroute, inputs [None {from=MaterialExpressionGetMaterialAttributes_0, output=Specular}]}
  MaterialExpressionScalarParameter_0 {class=MaterialExpressionScalarParameter, properties {defaultValue=2, parameterName=FlickerSpeed}}
  MaterialExpressionScalarParameter_1 {class=MaterialExpressionScalarParameter, properties {defaultValue=1, parameterName=WPOOffsetMultiplier}}
  MaterialExpressionScalarParameter_15 {class=MaterialExpressionScalarParameter, properties {defaultValue=1, parameterName="Number of Rows (S)"}}
  MaterialExpressionScalarParameter_16 {class=MaterialExpressionScalarParameter, properties {defaultValue=1, group=BisonDrainScreen, parameterName=FlipbookForTVTimeMultiplier}}
  MaterialExpressionScalarParameter_2 {class=MaterialExpressionScalarParameter, properties {defaultValue=1, parameterName="Number of Columns (S)"}}
  MaterialExpressionScalarParameter_3 {class=MaterialExpressionScalarParameter, properties {defaultValue=1, parameterName="Number of Rows (S)"}}
  MaterialExpressionScalarParameter_4 {class=MaterialExpressionScalarParameter, properties {defaultValue=0.8, parameterName=StrokeStrength}}
  MaterialExpressionScalarParameter_8 {class=MaterialExpressionScalarParameter, properties {defaultValue=1, parameterName="Number of Columns (S)"}}
  MaterialExpressionSetMaterialAttributes_0 {
   class=MaterialExpressionSetMaterialAttributes
   inputs [
    MaterialAttributes {from=MaterialExpressionReroute_0, output=""}
    BaseColor {from=MaterialExpressionNamedRerouteUsage_14, output=""}
    Roughness {from="", output=""}
    Specular {from=MaterialExpressionLinearInterpolate_6, output=""}
    Normal {from=MaterialExpressionNamedRerouteUsage_13, output=""}
   ]
   properties {attributeSetTypes [
     #1
     #2
     #3
     #4
    ]}
  }
  MaterialExpressionSmoothStep_1 {
   class=MaterialExpressionSmoothStep
   inputs [
    Min {from="", output=""}
    Max {from="", output=""}
    Value {from=MaterialExpressionDotProduct_1, output=""}
   ]
   properties {constMin=-0.5}
  }
  MaterialExpressionSubtract_2 {
   class=MaterialExpressionSubtract
   inputs [
    A {from=MaterialExpressionNamedRerouteUsage_12}
    B {from=""}
   ]
   properties {constB=0.5}
  }
  MaterialExpressionTextureCoordinate_0 {class=MaterialExpressionTextureCoordinate}
  MaterialExpressionTextureCoordinate_1 {class=MaterialExpressionTextureCoordinate}
  MaterialExpressionTextureObjectParameter_0 {
   class=MaterialExpressionTextureObjectParameter
   inputs [
    UVs {from="", output=""}
    Apply View MipBias {from="", output=""}
   ]
   properties {
    group=DustyBunTechArt
    parameterName=FlipbookForRandomLines
    samplerType=SAMPLERTYPE_LinearColor
    texture=/BRCosmetics/Characters/Player/Male/Large/Bodies/M_LRG_DustyBun/Textures/T_DustyBun_RandomBodyLines_FX2
   }
  }
  MaterialExpressionTextureObjectParameter_1 {
   class=MaterialExpressionTextureObjectParameter
   inputs [
    UVs {from="", output=""}
    Apply View MipBias {from="", output=""}
   ]
   properties {
    group=DustyBunTechArt
    parameterName=FlipbookForLines
    samplerType=SAMPLERTYPE_LinearColor
    texture=/BRCosmetics/Characters/Player/Male/Large/Bodies/M_LRG_DustyBun/Textures/T_DustyBun_Lines_FX
   }
  }
  MaterialExpressionTextureObjectParameter_2 {
   class=MaterialExpressionTextureObjectParameter
   inputs [
    UVs {from="", output=""}
    Apply View MipBias {from="", output=""}
   ]
   properties {
    group=DustyBunTechArt
    parameterName=FlipbookForLines2
    samplerType=SAMPLERTYPE_LinearColor
    texture=/BRCosmetics/Characters/Player/Male/Large/Bodies/M_LRG_DustyBun/Textures/T_DustyBun_Lines_FX2
   }
  }
  MaterialExpressionTextureObjectParameter_3 {
   class=MaterialExpressionTextureObjectParameter
   inputs [
    UVs {from="", output=""}
    Apply View MipBias {from="", output=""}
   ]
   properties {
    group=DustyBunTechArt
    parameterName=FlipbookForRandomLines2
    samplerType=SAMPLERTYPE_LinearColor
    texture=&T_DustyBun_RandomBodyLines_FX2
   }
  }
  MaterialExpressionTextureObjectParameter_4 {
   class=MaterialExpressionTextureObjectParameter
   inputs [
    UVs {from="", output=""}
    Apply View MipBias {from="", output=""}
   ]
   properties {
    group=DustyBunTechArt
    parameterName=FlipbookForLines
    texture=&T_DustyBun_Lines_FX
   }
  }
  MaterialExpressionTextureObjectParameter_5 {
   class=MaterialExpressionTextureObjectParameter
   inputs [
    UVs {from="", output=""}
    Apply View MipBias {from="", output=""}
   ]
   properties {
    group=DustyBunTechArt
    parameterName=FlipbookForLines2
    texture=&T_DustyBun_Lines_FX2
   }
  }
  MaterialExpressionTextureObjectParameter_6 {
   class=MaterialExpressionTextureObjectParameter
   inputs [
    UVs {from="", output=""}
    Apply View MipBias {from="", output=""}
   ]
   properties {
    group=DustyBunTechArt
    parameterName=FlipbookForRandomLines
    texture=&T_DustyBun_RandomBodyLines_FX2
   }
  }
  MaterialExpressionTextureSample_0 {
   class=MaterialExpressionTextureSample
   inputs [
    UVs {from=""}
    Tex {from=MaterialExpressionTextureObjectParameter_4}
    Apply View MipBias {from=""}
   ]
   properties {samplerSource=SSM_Wrap_WorldGroupSettings}
  }
  MaterialExpressionTextureSample_1 {
   class=MaterialExpressionTextureSample
   inputs [
    UVs {from=""}
    Tex {from=MaterialExpressionTextureObjectParameter_5}
    Apply View MipBias {from=""}
   ]
   properties {samplerSource=SSM_Wrap_WorldGroupSettings}
  }
  MaterialExpressionTextureSample_2 {
   class=MaterialExpressionTextureSample
   inputs [
    UVs {from=""}
    Tex {from=MaterialExpressionTextureObjectParameter_6}
    Apply View MipBias {from=""}
   ]
   properties {samplerSource=SSM_Wrap_WorldGroupSettings}
  }
  MaterialExpressionTime_0 {class=MaterialExpressionTime, properties {bOverride_Period=true, period=0.3333}}
  MaterialExpressionTime_1 {class=MaterialExpressionTime, properties {bOverride_Period=true, period=0.3833}}
  MaterialExpressionTime_2 {class=MaterialExpressionTime}
  MaterialExpressionTime_8 {class=MaterialExpressionTime}
  MaterialExpressionTransform_0 {
   class=MaterialExpressionTransform
   inputs [None {from=MaterialExpressionVertexNormalWS_0}]
   properties {transformSourceType=TRANSFORMSOURCE_World, transformType=TRANSFORM_Tangent}
  }
  MaterialExpressionVertexInterpolator_0 {class=MaterialExpressionVertexInterpolator, inputs [VS {from=MaterialExpressionLocalPosition_1, output=XYZ}], properties {}}
  MaterialExpressionVertexNormalWS_0 {class=MaterialExpressionVertexNormalWS, inputs [], properties {}}
 }}
