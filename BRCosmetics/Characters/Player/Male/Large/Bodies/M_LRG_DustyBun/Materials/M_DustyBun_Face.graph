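# Face material for the DustyBun large male player character, built on the 2023 character parent material functions with material attributes.
# The parent PreFX output runs through DustyBun's PBR-toon mix and body functions, then into the parent PostFX.
# A separate branch, not shown feeding PostFX, blends in 2023 hair shading by hair mask, with static switches for that shading and flyaway hair.
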Material /BRCosmetics/Characters/Player/Male/Large/Bodies/M_LRG_DustyBun/Materials/M_DustyBun_Face
graph {
 properties {
  bHasPixelAnimation=true
  bUseMaterialAttributes=true
  bUsedWithInstancedStaticMeshes=true
  bUsedWithMeshDeformer=true
  bUsedWithMorphTargets=true
  bUsedWithSkeletalMesh=true
  pixelDepthOffsetMode=PDOM_Legacy
  shadingModel=MSM_FromMaterialExpression
  usedShadingModels=Default Lit | Subsurface | Clear Coat | Subsurface Profile | Cloth
 }
 nodes [
  n1 = MaterialFunctionCall() {materialFunction=&MF_CharacterParent_2023_PreFX}
  n2 = NamedRerouteDeclaration(None=MaterialFunctionCall!(MA_In=@n1.Result, M=@n1.Result) {materialFunction=&/MF_DustyBun_Body_PBRToonMix}.MA_Out) {name=ToonPBRMix, nodeColor=<rgba 0.125 1 0 1>}
  MaterialFunctionCall(
   Material=MaterialFunctionCall!(In_MaterialAttributes=@n2) {materialFunction=&/MF_DustyBun_Body}.Result
   M=@n1.M
   Pre Skinned Local Position=@n1.M
   Distance Blend=@n1.M
   Fresnel=@n1.M
  ) {materialFunction=&MF_CharacterParent_2023_PostFX}
  n3 = MaterialFunctionCall!(In=@n2, Mask Input=@n1.M) {materialFunction=/Game/Global/Materials/Character_Parent_2023/Hair_Parent_2023/MaterialFunctions/MF_FN_Hair_2023}
  n4 = BlendMaterialAttributes(
   A=NamedRerouteUsage()
   B=StaticSwitchParameter(True=@n3.Result, False=@n1.Result) {defaultValue=true, group=Skin FX, parameterName=UseFNHair2_0}
   Alpha=@n3.HairMask
  )
  StaticSwitchParameter(
   True=MaterialFunctionCall!(MaterialInput=@n4) {materialFunction=/Game/Global/Materials/Character_Parent_2023/Hair_Parent_2023/MaterialFunctions/MF_FlyAwayHair_2023}.Result
   False=@n4
  ) {group=Skin FX, parameterName=IsFlyAwayHair?}
 ]
}
