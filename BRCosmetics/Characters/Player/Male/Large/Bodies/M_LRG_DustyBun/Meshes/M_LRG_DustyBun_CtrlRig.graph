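# Control Rig for the DustyBun large male body, correcting feet, arms and deformation bones over the animated pose.
# Pose-driven offsets move toe targets (with a frontend tweak), stretchy two-bone IK aims reparented foot chains at them, and foot-back bones get corrections.
# Unless emoting or wearing gauntlets, arm IK re-solves with pose-reader-shifted elbow pole vectors.
# A shared DeformRig function then runs, followed by spherical pose-reader clavicle correctives.
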ControlRigBlueprint /BRCosmetics/Characters/Player/Male/Large/Bodies/M_LRG_DustyBun/Meshes/M_LRG_DustyBun_CtrlRig
graphs [
 New Function_1_ContainedGraph {nodes [
   n1 = RigVMFunctionEntryNode()
   > RigVMFunction_Sequence_2 = RigVMAggregateNode()
   RigUnit_SetTransform = &RigUnit_SetTransform(
    Weight=1
    bInitial=false
    Value=&RigUnit_GetTransform(Item=(Type=Bone,Name="hand_l"), bInitial=false).Transform
    ExecutePin=@RigVMFunction_Sequence_2.A
    Item=@n1.Left Arm Controller
   )
   > &RigUnit_SetTransform(
    Weight=1
    bInitial=false
    Value=&RigUnit_GetTransform(Item=(Type=Bone,Name="hand_r"), bInitial=false).Transform
    Item=@n1.Right Arm Controller
   )
   > ComputePoleVector = RigVMFunctionReferenceNode(
    Bone A=(Type=Bone,Name="upperarm_l")
    Bone B=(Type=Bone,Name="lowerarm_l")
    Bone C=(Type=Bone,Name="hand_l")
    Initial=false
    OffsetFactor=5
    Origin Scale=(X=0,Y=0,Z=0)
    Draw Transform=@n1.Draw Debug
   )
   > &RigUnit_SetTransform(Weight=1, bInitial=false, Value=@ComputePoleVector.Transform, Item=@n1.Left Arm Pole Vector)
   > ComputePoleVector_1 = RigVMFunctionReferenceNode(
    Bone A=(Type=Bone,Name="upperarm_r")
    Bone B=(Type=Bone,Name="lowerarm_r")
    Bone C=(Type=Bone,Name="hand_r")
    Initial=false
    OffsetFactor=5
    Origin Scale=(X=0,Y=0,Z=0)
    Draw Transform=@n1.Draw Debug
   )
   > &RigUnit_SetTransform(Weight=1, bInitial=false, Value=@ComputePoleVector_1.Transform, Item=@n1.Right Arm Pole Vector)
   n2 = &RigUnit_SphericalPoseReader(
    ActiveRegionSize=0.1
    Debug=(bDrawDebug=true,bDraw2D=False,bDrawLocalAxes=False,DebugScale=25,DebugSegments=20,DebugThickness=0.25)
    DriverAxis=(X=1,Y=0,Z=0)
    DriverItem=(Type=Bone,Name="upperarm_l")
    FalloffSize=0.2
    FlipHeightScaling=False
    FlipWidthScaling=False
    OptionalParentItem=(Type=Bone,Name="spine_03")
    RotationOffset=(X=0,Y=-145,Z=0)
    ExecutePin=@RigVMFunction_Sequence_2.B
    FalloffRegionScaleFactors=@n1.Falloff Region Scale Factors
    Debug.bDrawDebug=@n1.Draw Debug
    ActiveRegionScaleFactors=@n1.Active Region Scale Factors
   )
   > &RigUnit_ModifyTransforms(
    ItemToModify="((Item=(Type=Control,Name=\"Arm_L_PV\"),Transform=(Rotation=(X=0,Y=0,Z=0,W=1),Translation=(X=0,Y=-123.5555,Z=0),Scale3D=(X=1,Y=1,Z=1))),(Item=(Type=Control,Name=\"Arm_L_IK\"),Transform=(Rotation=(X=0,Y=0,Z=0,W=1),Translation=(X=0,Y=0,Z=0),Scale3D=(X=1,Y=1,Z=1) ...(truncated)"
    WeightMinimum=0
    Weight=@n2.OutputParam
    ItemToModify.0.Transform.Translation=@n1.Left PV Add
   )
   > n3 = &RigUnit_SphericalPoseReader(
    ActiveRegionSize=0.1
    Debug=(bDrawDebug=true,bDraw2D=False,bDrawLocalAxes=False,DebugScale=25,DebugSegments=20,DebugThickness=0.25)
    DriverAxis=(X=-1,Y=0,Z=0)
    DriverItem=(Type=Bone,Name="upperarm_r")
    FalloffSize=0.2
    FlipHeightScaling=False
    FlipWidthScaling=False
    OptionalParentItem=(Type=Bone,Name="spine_03")
    RotationOffset=(X=0,Y=35,Z=0)
    FalloffRegionScaleFactors=@n1.Falloff Region Scale Factors
    Debug.bDrawDebug=@n1.Draw Debug
    ActiveRegionScaleFactors=@n1.Active Region Scale Factors
   )
   > &RigUnit_ModifyTransforms(
    ItemToModify=((Item=(Type=Control,Name="Arm_R_PV"),Transform=(Rotation=(X=0,Y=0,Z=0,W=-1),Translation=(X=0,Y=66.756,Z=0),Scale3D=(X=1,Y=1,Z=1))))
    WeightMinimum=0
    Weight=@n3.OutputParam
    ItemToModify.0.Transform.Translation=@n1.Right PV Add
   )
   /Script/ControlRig.RigUnit_TwoBoneIKSimplePerItem(
    DebugSettings=(bEnabled=False,Scale=10,WorldOffset=(Rotation=(X=0,Y=0,Z=0,W=1),Translation=(X=0,Y=0,Z=0),Scale3D=(X=1,Y=1,Z=1)))
    EffectorItem=(Type=Bone,Name="hand_l")
    ItemA=(Type=Bone,Name="upperarm_l")
    ItemALength=0
    ItemB=(Type=Bone,Name="lowerarm_l")
    ItemBLength=0
    PoleVectorKind=Location
    PoleVectorSpace=(Type=Bone,Name="None")
    PrimaryAxis=(X=1,Y=0,Z=0)
    SecondaryAxis=(X=0,Y=-1,Z=0)
    SecondaryAxisWeight=1
    StretchMaximumRatio=1.25
    StretchStartRatio=1
    Weight=1
    bEnableStretch=false
    bPropagateToChildren=True
    Effector=&RigUnit_GetTransform(Item=(Type=Control,Name="Arm_L_IK"), bInitial=False).Transform
    PoleVector=&RigUnit_GetTransform(Item=(Type=Control,Name="Arm_L_PV"), bInitial=False).Transform.Translation
    ExecutePin=@RigVMFunction_Sequence_2.C
   ) {is=struct RigVMUnitNode}
   > &RigUnit_TwoBoneIKSimplePerItem(
    DebugSettings=(bEnabled=False,Scale=10,WorldOffset=(Rotation=(X=0,Y=0,Z=0,W=1),Translation=(X=0,Y=0,Z=0),Scale3D=(X=1,Y=1,Z=1)))
    EffectorItem=(Type=Bone,Name="hand_r")
    ItemA=(Type=Bone,Name="upperarm_r")
    ItemALength=0
    ItemB=(Type=Bone,Name="lowerarm_r")
    ItemBLength=0
    PoleVectorKind=Location
    PoleVectorSpace=(Type=Bone,Name="None")
    PrimaryAxis=(X=-1,Y=0,Z=0)
    SecondaryAxis=(X=0,Y=1,Z=0)
    SecondaryAxisWeight=1
    StretchMaximumRatio=1.25
    StretchStartRatio=1
    Weight=1
    bEnableStretch=false
    bPropagateToChildren=True
    Effector=&RigUnit_GetTransform(Item=(Type=Control,Name="Arm_R_IK"), bInitial=False).Transform
    PoleVector=&RigUnit_GetTransform(Item=(Type=Control,Name="Arm_R_PV"), bInitial=False).Transform.Translation
   )
   RigVMFunctionReturnNode(Right Pose Weight=@n3.OutputParam, Left Pose Weight=@n2.OutputParam, ExecuteContext=@RigVMFunction_Sequence_2.D)
  ]}
 RigVMFunctionLibrary {nodes [Adjust IK Elbows for M_LRG = RigVMCollapseNode(
    Active Region Scale Factors=(PositiveWidth=2,NegativeWidth=0.5,PositiveHeight=1,NegativeHeight=1)
    Draw Debug=true
    ExecuteContext=()
    Falloff Region Scale Factors=(PositiveWidth=2,NegativeWidth=1,PositiveHeight=2,NegativeHeight=2)
    Left Arm Controller=(Type=Control,Name="Arm_L_IK")
    Left Arm Pole Vector=(Type=Control,Name="Arm_L_PV")
    Left PV Add=(X=0,Y=-123.5555,Z=0)
    Right Arm Controller=(Type=Control,Name="Arm_R_IK")
    Right Arm Pole Vector=(Type=Control,Name="Arm_R_PV")
    Right PV Add=(X=0,Y=66.756,Z=0)
   )]}
 RigVMFunction_Sequence_2_ContainedGraph {nodes [
   n1 = RigVMFunctionEntryNode()
   RigVMFunction_Sequence = &RigVMFunction_Sequence(ExecuteContext=@n1.ExecuteContext)
   n2 = &RigVMFunction_Sequence(ExecuteContext=@RigVMFunction_Sequence.B)
   n3 = &RigVMFunction_Sequence(ExecuteContext=@n2.B)
   n4 = &RigVMFunction_Sequence(ExecuteContext=@n3.B)
   n5 = &RigVMFunction_Sequence(ExecuteContext=@n4.B)
   RigVMFunctionReturnNode(A=@RigVMFunction_Sequence.A, B=@n2.A, C=@n3.A, D=@n4.A, E=@n5.A, F=@n5.B)
  ]}
 RigVMFunction_Sequence_3_ContainedGraph {nodes [
   n1 = RigVMFunctionEntryNode()
   n2 = &RigVMFunction_Sequence(ExecuteContext=@n1.ExecuteContext)
   n3 = &RigVMFunction_Sequence(ExecuteContext=@n2.B)
   n4 = &RigVMFunction_Sequence(ExecuteContext=@n3.B)
   RigVMFunctionReturnNode(B=@n3.A, A=@n2.A, C=@n4.A, D=@n4.B)
  ]}
 RigVMModel {nodes [
   RigUnit_BeginExecution = &RigUnit_BeginExecution()
   RigVMFunction_Sequence = RigVMAggregateNode(ExecuteContext=@RigUnit_BeginExecution.ExecutePin)
   RigUnit_GetTransform = &RigUnit_GetTransform(Item=(Type=Bone,Name="ctrl_ball_end_r"), bInitial=False)
   &RigUnit_TwoBoneIKSimplePerItem(
    DebugSettings=(bEnabled=False,Scale=10,WorldOffset=(Rotation=(X=0,Y=0,Z=0,W=1),Translation=(X=0,Y=0,Z=0),Scale3D=(X=1,Y=1,Z=1)))
    EffectorItem=(Type=Bone,Name="ball_end_r")
    ItemA=(Type=Bone,Name="reparent_foot_r")
    ItemALength=0
    ItemB=(Type=Bone,Name="reparent_ball_r")
    ItemBLength=0
    PoleVector=(X=103.2294,Y=-42.2344,Z=13.4039)
    PoleVectorKind=Location
    PoleVectorSpace=(Type=Bone,Name="foot_r")
    PrimaryAxis=(X=-1,Y=0,Z=0)
    SecondaryAxis=(X=0,Y=1,Z=0)
    SecondaryAxisWeight=1
    StretchMaximumRatio=1.5
    StretchStartRatio=1
    Weight=1
    bEnableStretch=true
    bPropagateToChildren=True
    Effector=@RigUnit_GetTransform.Transform
    ExecutePin=@RigVMFunction_Sequence.B
   )
   > &RigUnit_TwoBoneIKSimplePerItem(
    DebugSettings=(bEnabled=False,Scale=10,WorldOffset=(Rotation=(X=0,Y=0,Z=0,W=1),Translation=(X=0,Y=0,Z=0),Scale3D=(X=1,Y=1,Z=1)))
    EffectorItem=(Type=Bone,Name="ball_end_l")
    ItemA=(Type=Bone,Name="reparent_foot_l")
    ItemALength=0
    ItemB=(Type=Bone,Name="reparent_ball_l")
    ItemBLength=0
    PoleVector=(X=-124.0291,Y=0,Z=0)
    PoleVectorKind=Location
    PoleVectorSpace=(Type=Bone,Name="foot_l")
    PrimaryAxis=(X=1,Y=0,Z=0)
    SecondaryAxis=(X=0,Y=-1,Z=0)
    SecondaryAxisWeight=1
    StretchMaximumRatio=1.5
    StretchStartRatio=1
    Weight=1
    bEnableStretch=true
    bPropagateToChildren=True
    Effector=&RigUnit_GetTransform(Item=(Type=Bone,Name="ctrl_ball_end_l"), bInitial=False).Transform
   )
   n1 = &RigUnit_SphericalPoseReader(
    ActiveRegionScaleFactors=(PositiveWidth=1,NegativeWidth=1,PositiveHeight=2,NegativeHeight=2)
    ActiveRegionSize=0.1
    Debug=(bDrawDebug=True,bDraw2D=False,bDrawLocalAxes=False,DebugScale=25,DebugSegments=20,DebugThickness=0.25)
    DriverAxis=(X=-1,Y=0,Z=0)
    DriverItem=(Type=Bone,Name="reparent_foot_r")
    FalloffRegionScaleFactors=(PositiveWidth=1,NegativeWidth=1,PositiveHeight=1,NegativeHeight=1)
    FalloffSize=0.4
    FlipHeightScaling=False
    FlipWidthScaling=False
    OptionalParentItem=(Type=Bone,Name="None")
    RotationOffset=(X=90,Y=-0.2908,Z=-8.4086)
    ExecutePin=@RigVMFunction_Sequence.C
   )
   > &RigUnit_ModifyTransforms(
    ItemToModify=((Item=(Type=Bone,Name="dyn_reparent_foot_bk_r"),Transform=(Rotation=(X=0.4189,Y=0,Z=0,W=0.908),Translation=(X=0,Y=-4.2522,Z=-5.5831),Scale3D=(X=1,Y=1,Z=1))))
    WeightMinimum=0
    Weight=@n1.OutputParam
   )
   > n2 = &RigUnit_SphericalPoseReader(
    ActiveRegionScaleFactors=(PositiveWidth=1,NegativeWidth=1,PositiveHeight=2,NegativeHeight=2)
    ActiveRegionSize=0.1
    Debug=(bDrawDebug=True,bDraw2D=False,bDrawLocalAxes=False,DebugScale=25,DebugSegments=20,DebugThickness=0.25)
    DriverAxis=(X=-1,Y=0,Z=0)
    DriverItem=(Type=Bone,Name="reparent_foot_l")
    FalloffRegionScaleFactors=(PositiveWidth=1,NegativeWidth=1,PositiveHeight=1,NegativeHeight=1)
    FalloffSize=0.4
    FlipHeightScaling=False
    FlipWidthScaling=False
    OptionalParentItem=(Type=Bone,Name="None")
    RotationOffset=(X=90,Y=-0.2908,Z=-8.4086)
   )
   > &RigUnit_ModifyTransforms(
    ItemToModify=((Item=(Type=Bone,Name="dyn_reparent_foot_bk_l"),Transform=(Rotation=(X=0.4189,Y=0,Z=0,W=0.908),Translation=(X=0,Y=-4.2522,Z=-5.5831),Scale3D=(X=1,Y=1,Z=1))))
    WeightMinimum=0
    Weight=@n2.OutputParam
   )
   RigVMFunction_MathBoolOr = &RigVMFunction_MathBoolOr(A=RigVMVariableNode() {variable=isPlayingEmote}.Value, B=RigVMVariableNode() {variable=isGauntletsEquip}.Value)
   RigVMFunction_ControlFlowBranch = &RigVMFunction_ControlFlowBranch(Condition=@RigVMFunction_MathBoolOr.Result, ExecuteContext=@RigVMFunction_Sequence.D)
   Adjust IK Elbows for M_LRG = RigVMFunctionReferenceNode(
    Active Region Scale Factors=(PositiveWidth=2,NegativeWidth=0.5,PositiveHeight=1,NegativeHeight=1)
    Draw Debug=false
    Falloff Region Scale Factors=(PositiveWidth=2,NegativeWidth=1.5,PositiveHeight=2,NegativeHeight=2)
    Left Arm Controller=(Type=Control,Name="Arm_L_IK")
    Left Arm Pole Vector=(Type=Control,Name="Arm_L_PV")
    Left PV Add=(X=0,Y=-123.5555,Z=0)
    Right Arm Controller=(Type=Control,Name="Arm_R_IK")
    Right Arm Pole Vector=(Type=Control,Name="Arm_R_PV")
    Right PV Add=(X=0,Y=66.756,Z=0)
    ExecuteContext=@RigVMFunction_ControlFlowBranch.False
   )
   FN_DeformRig_FNC = RigVMFunctionReferenceNode(
    DeformRig_ Curve="(deformRig_l_upperarm_up_alpha_72_#1=False,deformRig_r_upperarm_up_alpha_73_#2=False,deformRig_l_upperarm_fr_alpha_74_#3=False,deformRig_r_upperarm_fr_alpha_75_CB51B8 ...(truncated)"
    DeformRig_Bool="(upperarm_1_#4=True,elbow_3_#5=True,wrist_7_#6=True,groin_8_#7=True,glute_17_#8=True,knee_9_40C408324CA261 ...(truncated)"
    DeformRig_Debug=(upperarm_debug_18_#4=False,wrist_debug_20_#6=False,lat_debug_26_#9=False,groinGlute_debug_27_#7=False)
    DeformRig_Value="(upperarm_up_18_#10=(Rotation=(X=0,Y=0,Z=0,W=1),Translation=(X=0,Y=0,Z=8),Scale3D=(X=0.7,Y=1,Z=1)),upperarm_fwd_fr_19_#11=(Rotation=(X=0,Y=0,Z=0,W=1),Translation=(X=0,Y=10,Z=0),Scale3D=(X=0.5,Y=1,Z= ...(truncated)"
    ExecuteContext=@RigVMFunction_Sequence.E
   )
   n3 = &RigUnit_SphericalPoseReader(
    ActiveRegionScaleFactors=(PositiveWidth=2,NegativeWidth=1,PositiveHeight=1,NegativeHeight=1)
    ActiveRegionSize=0.1
    Debug=(bDrawDebug=true,bDraw2D=False,bDrawLocalAxes=False,DebugScale=25,DebugSegments=20,DebugThickness=0.25)
    DriverAxis=(X=1,Y=0,Z=0)
    DriverItem=(Type=Bone,Name="upperarm_r")
    FalloffRegionScaleFactors=(PositiveWidth=2,NegativeWidth=1,PositiveHeight=1,NegativeHeight=1)
    FalloffSize=0.3
    FlipHeightScaling=False
    FlipWidthScaling=False
    OptionalParentItem=(Type=Bone,Name="spine_05")
    RotationOffset=(X=100.1787,Y=-31.2267,Z=22.7226)
    ExecutePin=@RigVMFunction_Sequence.F
   )
   > &RigUnit_ModifyTransforms(
    ItemToModify=((Item=(Type=Bone,Name="deform_clavicle_r"),Transform=(Rotation=(X=-0.1259,Y=-0.1865,Z=-0.545,W=0.8077),Translation=(X=13.0375,Y=4.5287,Z=-3.3399),Scale3D=(X=1,Y=1,Z=1))))
    WeightMinimum=0
    Weight=@n3.OutputParam
   )
   > n4 = &RigUnit_SphericalPoseReader(
    ActiveRegionScaleFactors=(PositiveWidth=2,NegativeWidth=1,PositiveHeight=1,NegativeHeight=1)
    ActiveRegionSize=0.1
    Debug=(bDrawDebug=true,bDraw2D=False,bDrawLocalAxes=False,DebugScale=25,DebugSegments=20,DebugThickness=0.25)
    DriverAxis=(X=1,Y=0,Z=0)
    DriverItem=(Type=Bone,Name="upperarm_l")
    FalloffRegionScaleFactors=(PositiveWidth=2,NegativeWidth=1,PositiveHeight=1,NegativeHeight=1)
    FalloffSize=0.3
    FlipHeightScaling=False
    FlipWidthScaling=False
    OptionalParentItem=(Type=Bone,Name="spine_05")
    RotationOffset=(X=100.1787,Y=-31.2267,Z=22.7226)
   )
   > &RigUnit_ModifyTransforms(
    ItemToModify=((Item=(Type=Bone,Name="deform_clavicle_l"),Transform=(Rotation=(X=0.1259,Y=0.1865,Z=-0.545,W=0.8077),Translation=(X=13.0375,Y=-4.5287,Z=-3.3399),Scale3D=(X=1,Y=1,Z=1))))
    WeightMinimum=0
    Weight=@n4.OutputParam
   )
   RigVMFunction_MathBoolNot = &RigVMFunction_MathBoolNot(Value=RigVMVariableNode() {variable=isPlayingEmote}.Value)
   RigVMFunction_MathBoolAnd = &RigVMFunction_MathBoolAnd(A=@RigVMFunction_MathBoolNot.Result, B=RigVMVariableNode() {variable=isFrontend}.Value)
   RigVMFunction_MathBoolToFloat = &RigVMFunction_MathBoolToFloat(Value=@RigVMFunction_MathBoolAnd.Result)
   Multiply = &RigVMFunction_MathDoubleMul(B=-1, A=@RigVMFunction_MathBoolToFloat.Result)
   &RigUnit_ModifyTransforms(
    ItemToModify="((Item=(Type=Bone,Name=\"ctrl_ball_end_r\"),Transform=(Rotation=(X=0,Y=0,Z=0,W=-1),Translation=(X=1,Y=0,Z=0),Scale3D=(X=1,Y=1,Z=1))),(Item=(Type=Bone,Name=\"ctrl_ball_end_l\"),Transform=(Rotation=(X=0,Y=0,Z=0,W=1),Translation=(X=0,Y=0,Z=0),Scale3D=(X=1,Y=1,Z=1 ...(truncated)"
    Weight=1
    WeightMinimum=0
    ItemToModify.1.Transform.Translation.X=@Multiply.Result
    ExecutePin=@RigVMFunction_Sequence.A
   )
   > n5 = &RigUnit_SphericalPoseReader(
    ActiveRegionScaleFactors=(PositiveWidth=1,NegativeWidth=1,PositiveHeight=1,NegativeHeight=1)
    ActiveRegionSize=0.1
    Debug=(bDrawDebug=False,bDraw2D=False,bDrawLocalAxes=False,DebugScale=25,DebugSegments=20,DebugThickness=0.25)
    DriverAxis=(X=-1,Y=0,Z=0)
    DriverItem=(Type=Bone,Name="ball_r")
    FalloffRegionScaleFactors=(PositiveWidth=1,NegativeWidth=1,PositiveHeight=1,NegativeHeight=1)
    FalloffSize=0.2
    FlipHeightScaling=False
    FlipWidthScaling=False
    OptionalParentItem=(Type=Bone,Name="calf_r")
    RotationOffset=(X=90,Y=0,Z=0)
   )
   > &RigUnit_ModifyTransforms(
    ItemToModify=((Item=(Type=Bone,Name="ctrl_ball_end_r"),Transform=(Rotation=(X=-0.0045,Y=0,Z=0,W=1),Translation=(X=4.4,Y=-2.708,Z=0),Scale3D=(X=1,Y=1,Z=1))))
    WeightMinimum=0
    Weight=@n5.OutputParam
   )
   > n6 = &RigUnit_SphericalPoseReader(
    ActiveRegionScaleFactors=(PositiveWidth=1,NegativeWidth=1,PositiveHeight=1,NegativeHeight=1)
    ActiveRegionSize=0.1
    Debug=(bDrawDebug=False,bDraw2D=False,bDrawLocalAxes=False,DebugScale=25,DebugSegments=20,DebugThickness=0.25)
    DriverAxis=(X=-1,Y=0,Z=0)
    DriverItem=(Type=Bone,Name="ball_l")
    FalloffRegionScaleFactors=(PositiveWidth=1,NegativeWidth=1,PositiveHeight=1,NegativeHeight=1)
    FalloffSize=0.2
    FlipHeightScaling=False
    FlipWidthScaling=False
    OptionalParentItem=(Type=Bone,Name="calf_l")
    RotationOffset=(X=90,Y=0,Z=0)
   )
   > &RigUnit_ModifyTransforms(
    ItemToModify=((Item=(Type=Bone,Name="ctrl_ball_end_l"),Transform=(Rotation=(X=0,Y=0,Z=0,W=1),Translation=(X=-4.4089,Y=2.708,Z=0),Scale3D=(X=1,Y=1,Z=1))))
    WeightMinimum=0
    Weight=@n6.OutputParam
   )
  ]}
]
hierarchy [
 root {loc=<0 0 0>, rot=<0 0 0 1>}
 pelvis {parent=root, loc=<0 -0.7321 92.4218>, rot=<0.0127 0.707 -0.0127 -0.707>}
 spine_01 {parent=pelvis, loc=<5.6002 -0.5128 0>, rot=<0 0 0.034 -0.9994>}
 spine_02 {parent=spine_01, loc=<8.0753 0.0139 0>, rot=<0 0 0.0238 -0.9997>}
 spine_03 {parent=spine_02, loc=<7.0393 0 0>, rot=<0 0 0.0149 -0.9999>}
 spine_04 {parent=spine_03, loc=<8.1024 0 0>, rot=<0 0 -0.0132 -0.9999>}
 spine_05 {parent=spine_04, loc=<10.0146 0 0>, rot=<0 0 -0.0015 -1>}
 clavicle_l {parent=spine_05, loc=<2.5856 -0.2866 -4.3777>, rot=<0.0465 -0.7289 -0.009 -0.6829>}
 upperarm_l {parent=clavicle_l, loc=<18.2264 0.1342 -0.001>, rot=<0.0253 -0.2693 0.0013 -0.9627>}
 lowerarm_l {parent=upperarm_l, loc=<29.0071 0.1686 -1.239>, rot=<0.0391 -0.0099 -0.2453 -0.9686>}
 hand_l {parent=lowerarm_l, loc=<25.8607 0.222 1.0274>, rot=<-0.7193 -0.0988 0.1193 -0.6773>}
 index_metacarpal_l {parent=hand_l, loc=<1.6535 0.0183 -2.96>, rot=<0.0158 -0.0124 -0.0173 -0.9996>}
 index_01_l {parent=index_metacarpal_l, loc=<8.3008 0.0357 -0.0524>, rot=<-0.0044 0.0485 0.0924 -0.9945>}
 index_02_l {parent=index_01_l, loc=<3.3587 -0.3004 0.0452>, rot=<-0.002 0.0027 0.3225 -0.9466>}
 index_03_l {parent=index_02_l, loc=<2.5578 -0.123 -0.0252>, rot=<-0.002 0.0067 0.2558 -0.9667>}
 middle_metacarpal_l {parent=hand_l, loc=<2.1856 -0.1725 -0.5984>, rot=<0.006 0.0168 -0.0329 -0.9993>}
 middle_01_l {parent=middle_metacarpal_l, loc=<7.3275 -0.2447 0.0129>, rot=<-0.0208 0.0042 0.0858 -0.9961>}
 middle_02_l {parent=middle_01_l, loc=<4.0123 0.1401 0.1175>, rot=<-0.0218 0.0258 0.3512 -0.9357>}
 middle_03_l {parent=middle_02_l, loc=<2.6077 0.0534 0.1373>, rot=<0.0082 0.0188 0.1708 -0.9851>}
 pinky_metacarpal_l {parent=hand_l, loc=<1.7616 -0.9753 3.3814>, rot=<-0.0397 0.0814 -0.069 -0.9935>}
 pinky_01_l {parent=pinky_metacarpal_l, loc=<7.1709 -0.4831 -0.1094>, rot=<-0.0654 -0.0035 0.2242 -0.9723>}
 pinky_02_l {parent=pinky_01_l, loc=<2.8274 0.0893 -0.1043>, rot=<0.0037 0.0011 0.2854 -0.9584>}
 pinky_03_l {parent=pinky_02_l, loc=<2.2046 0.0307 -0.0078>, rot=<-0.0002 0.0098 0.1445 -0.9895>}
 ring_metacarpal_l {parent=hand_l, loc=<2.2074 -0.3625 1.3638>, rot=<-0.0284 0.0722 -0.0458 -0.9959>}
 ring_01_l {parent=ring_metacarpal_l, loc=<7.0295 -0.6197 -0.1446>, rot=<-0.0489 -0.0252 0.1892 -0.9804>}
 ring_02_l {parent=ring_01_l, loc=<3.6174 0.4647 -0.1027>, rot=<0 0 0.2806 -0.9598>}
 ring_03_l {parent=ring_02_l, loc=<2.4152 0.114 0.0199>, rot=<0 0 0.1189 -0.9929>}
 thumb_01_l {parent=hand_l, loc=<1.3118 -0.3191 -2.459>, rot=<0.6546 -0.4042 0.0664 -0.6354>}
 thumb_02_l {parent=thumb_01_l, loc=<5.9336 -0.0448 0.0401>, rot=<-0.0037 0.0213 0.298 -0.9543>}
 thumb_03_l {parent=thumb_02_l, loc=<2.2223 0.3682 -0.0907>, rot=<0.0118 -0.013 0.0066 -0.9998>}
 deform_wrist_dn_base_l {parent=hand_l, loc=<0 -3.4173 0>, rot=<0 0 0 1>}
 deform_wrist_dn_l {parent=deform_wrist_dn_base_l, loc=<0 0 0>, rot=<0 0 0 1>}
 deform_wrist_up_base_l {parent=hand_l, loc=<0 2.7805 0>, rot=<0 0 0 1>}
 deform_wrist_up_l {parent=deform_wrist_up_base_l, loc=<0 0 0>, rot=<0 0 0 1>}
 lowerarm_twist_01_l {parent=lowerarm_l, loc=<17.8549 0 0>, rot=<0 0 0 1>}
 lowerarm_twist_02_l {parent=lowerarm_l, loc=<9.6199 0 0>, rot=<0 0 0 1>}
 deform_elbow_in_base_l {parent=lowerarm_l, loc=<0.2112 5.645 0.8274>, rot=<0 0 0 1>}
 deform_elbow_in_l {parent=deform_elbow_in_base_l, loc=<0 0 0>, rot=<0 0 0 1>}
 deform_elbow_out_base_l {parent=lowerarm_l, loc=<0 -7.5129 0>, rot=<0 0 0 1>}
 deform_elbow_out_l {parent=deform_elbow_out_base_l, loc=<0 0 0>, rot=<0 0 0 1>}
 upperarm_twist_01_l {parent=upperarm_l, loc=<7.9796 0.0843 -0.6195>, rot=<0 0 0 1>}
 deform_upperarm_bk_base_l {parent=upperarm_twist_01_l, loc=<-6.8983 -13.3826 -1.3258>, rot=<0 0 0 1>}
 deform_upperarm_bk_l {parent=deform_upperarm_bk_base_l, loc=<0 0 0>, rot=<0 0 0 1>}
 deform_upperarm_fr_base_l {parent=upperarm_twist_01_l, loc=<-8.3487 10.7117 -2.1455>, rot=<0 0 0 1>}
 deform_upperarm_fr_l {parent=deform_upperarm_fr_base_l, loc=<0 0 0>, rot=<0 0 0 1>}
 deform_upperarm_up_base_l {parent=upperarm_twist_01_l, loc=<-11.2012 -1.2234 8.7784>, rot=<0 0 0 1>}
 deform_upperarm_up_l {parent=deform_upperarm_up_base_l, loc=<0 0 0>, rot=<0 0 0 1>}
 upperarm_twist_02_l {parent=upperarm_l, loc=<19.5059 0.0843 -0.6195>, rot=<0 0 0 1>}
 deform_bicep_base_l {parent=upperarm_twist_02_l, loc=<0 8.0494 0>, rot=<0 0 0 1>}
 deform_bicep_l {parent=deform_bicep_base_l, loc=<0 0 0>, rot=<0 0 0 1>}
 deform_clavicle_l_base {parent=clavicle_l, loc=<15.2305 9.7776 5.4596>, rot=<0.0243 0.0231 -0.7243 -0.6887>}
 deform_clavicle_l {parent=deform_clavicle_l_base, loc=<0 0 0>, rot=<0 0 0 1>}
 clavicle_r {parent=spine_05, loc=<2.5856 -0.2865 4.3777>, rot=<0.7289 0.0465 -0.6829 0.009>}
 upperarm_r {parent=clavicle_r, loc=<-18.2264 -0.1342 0.001>, rot=<0.0253 -0.2693 0.0013 -0.9627>}
 lowerarm_r {parent=upperarm_r, loc=<-29.0072 -0.1686 1.239>, rot=<0.0391 -0.0099 -0.2453 -0.9686>}
 hand_r {parent=lowerarm_r, loc=<-25.8608 -0.222 -1.0274>, rot=<-0.7193 -0.0988 0.1193 -0.6773>}
 index_metacarpal_r {parent=hand_r, loc=<-1.6534 -0.0181 2.9601>, rot=<0.0158 -0.0124 -0.0173 -0.9996>}
 index_01_r {parent=index_metacarpal_r, loc=<-8.3008 -0.0357 0.0524>, rot=<-0.0044 0.0485 0.0924 -0.9945>}
 index_02_r {parent=index_01_r, loc=<-3.3589 0.3003 -0.0453>, rot=<-0.002 0.0027 0.3225 -0.9466>}
 index_03_r {parent=index_02_r, loc=<-2.5578 0.123 0.0252>, rot=<-0.002 0.0067 0.2558 -0.9667>}
 middle_metacarpal_r {parent=hand_r, loc=<-2.1856 0.1726 0.5984>, rot=<0.006 0.0168 -0.0329 -0.9993>}
 middle_01_r {parent=middle_metacarpal_r, loc=<-7.3273 0.2447 -0.0129>, rot=<-0.0208 0.0042 0.0858 -0.9961>}
 middle_02_r {parent=middle_01_r, loc=<-4.0124 -0.14 -0.1175>, rot=<-0.0218 0.0258 0.3512 -0.9357>}
 middle_03_r {parent=middle_02_r, loc=<-2.6077 -0.0533 -0.1373>, rot=<0.0082 0.0188 0.1708 -0.9851>}
 pinky_metacarpal_r {parent=hand_r, loc=<-1.7617 0.9753 -3.3814>, rot=<-0.0397 0.0814 -0.069 -0.9935>}
 pinky_01_r {parent=pinky_metacarpal_r, loc=<-7.1707 0.4832 0.1094>, rot=<-0.0654 -0.0035 0.2242 -0.9723>}
 pinky_02_r {parent=pinky_01_r, loc=<-2.8274 -0.0892 0.1043>, rot=<0.0037 0.0011 0.2854 -0.9584>}
 pinky_03_r {parent=pinky_02_r, loc=<-2.2045 -0.031 0.0078>, rot=<-0.0002 0.0098 0.1445 -0.9895>}
 ring_metacarpal_r {parent=hand_r, loc=<-2.2073 0.3627 -1.3638>, rot=<-0.0284 0.0722 -0.0458 -0.9959>}
 ring_01_r {parent=ring_metacarpal_r, loc=<-7.0297 0.6197 0.1446>, rot=<-0.0489 -0.0252 0.1892 -0.9804>}
 ring_02_r {parent=ring_01_r, loc=<-3.6171 -0.4648 0.1027>, rot=<0 0 0.2806 -0.9598>}
 ring_03_r {parent=ring_02_r, loc=<-2.4153 -0.1141 -0.0199>, rot=<0 0 0.1189 -0.9929>}
 thumb_01_r {parent=hand_r, loc=<-1.3119 0.3192 2.459>, rot=<0.6546 -0.4042 0.0664 -0.6354>}
 thumb_02_r {parent=thumb_01_r, loc=<-5.9335 0.0447 -0.0401>, rot=<-0.0037 0.0213 0.298 -0.9543>}
 thumb_03_r {parent=thumb_02_r, loc=<-2.2223 -0.3682 0.0907>, rot=<0.0118 -0.013 0.0066 -0.9998>}
 deform_wrist_dn_base_r {parent=hand_r, loc=<0.0001 3.4173 0>, rot=<0 0 0 -1>}
 deform_wrist_dn_r {parent=deform_wrist_dn_base_r, loc=<0 0 0>, rot=<0 0 0 -1>}
 deform_wrist_up_base_r {parent=hand_r, loc=<0 -2.7803 0>, rot=<0 0 0 -1>}
 deform_wrist_up_r {parent=deform_wrist_up_base_r, loc=<0 0 0>, rot=<0 0 0 -1>}
 lowerarm_twist_01_r {parent=lowerarm_r, loc=<-17.855 0 0>, rot=<0 0 0 1>}
 lowerarm_twist_02_r {parent=lowerarm_r, loc=<-9.6199 0 0>, rot=<0 0 0 1>}
 deform_elbow_in_base_r {parent=lowerarm_r, loc=<-0.2114 -5.6449 -0.827>, rot=<0 0 0 1>}
 deform_elbow_in_r {parent=deform_elbow_in_base_r, loc=<0 0 0>, rot=<0 0 0 1>}
 deform_elbow_out_base_r {parent=lowerarm_r, loc=<-0.0001 7.5129 0.0002>, rot=<0 0 0 1>}
 deform_elbow_out_r {parent=deform_elbow_out_base_r, loc=<0 0 0>, rot=<0 0 0 1>}
 upperarm_twist_01_r {parent=upperarm_r, loc=<-7.9796 -0.0843 0.6195>, rot=<0 0 0 1>}
 deform_upperarm_bk_base_r {parent=upperarm_twist_01_r, loc=<6.8984 13.3826 1.3257>, rot=<0 0 0 1>}
 deform_upperarm_bk_r {parent=deform_upperarm_bk_base_r, loc=<0 0 0>, rot=<0 0 0 1>}
 deform_upperarm_fr_base_r {parent=upperarm_twist_01_r, loc=<8.3485 -10.7117 2.1457>, rot=<0 0 0 1>}
 deform_upperarm_fr_r {parent=deform_upperarm_fr_base_r, loc=<0 0 0>, rot=<0 0 0 1>}
 deform_upperarm_up_base_r {parent=upperarm_twist_01_r, loc=<11.2014 1.2235 -8.7787>, rot=<0 0 0 1>}
 deform_upperarm_up_r {parent=deform_upperarm_up_base_r, loc=<0 0 0>, rot=<0 0 0 1>}
 upperarm_twist_02_r {parent=upperarm_r, loc=<-19.506 -0.0843 0.6195>, rot=<0 0 0 1>}
 deform_bicep_base_r {parent=upperarm_twist_02_r, loc=<0.0002 -8.0494 -0.0002>, rot=<0 0 0 1>}
 deform_bicep_r {parent=deform_bicep_base_r, loc=<0 0 0>, rot=<0 0 0 1>}
 deform_clavicle_r_base {parent=clavicle_r, loc=<-15.2307 -9.7776 -5.4595>, rot=<0.7243 -0.6887 0.0243 -0.0231>}
 deform_clavicle_r {parent=deform_clavicle_r_base, loc=<0 0 0>, rot=<0 0 0 1>}
 neck_01 {parent=spine_05, loc=<10.815 1.9196 0>, rot=<0 0 -0.1864 -0.9825>}
 neck_02 {parent=neck_01, loc=<4.7347 0.3045 0>, rot=<0 0 0.0363 -0.9993>}
 head {parent=neck_02, loc=<5.7632 0.5552 0>, rot=<0 0 0.115 -0.9934>}
 attach_cape {parent=spine_05, loc=<9.017 -11.714 0>, rot=<0.0616 -0.7044 -0.0616 -0.7044>}
 cape {parent=attach_cape, type=Null, loc=<0 -6.3579 0>, rot=<0 0 0 1>}
 attach_backpack {parent=spine_05, loc=<-0.123 -13.473 0>, rot=<-0.0093 -0.707 0.0093 -0.707>}
 Backpack_BR {parent=attach_backpack, type=Null, loc=<0 -7.2428 0>, rot=<0.0331 0 0 0.9995>}
 deform_pec_aim_base_r {parent=spine_04, loc=<5.023 13.6718 2.6508>, rot=<-0.5201 -0.0672 0.7879 0.3228>}
 deform_pec_aim_r {parent=deform_pec_aim_base_r, loc=<0 0 0>, rot=<0 0 0 1>}
 deform_pec_base_r {parent=deform_pec_aim_r, loc=<-16.3279 0 0>, rot=<0 0 0 -1>}
 deform_pec_r {parent=deform_pec_base_r, loc=<0 0 0.0002>, rot=<0 0 0 -1>}
 deform_pec_aim_base_l {parent=spine_04, loc=<5.0232 13.6719 -2.6508>, rot=<0.0672 -0.5201 0.3228 -0.7879>}
 deform_pec_aim_l {parent=deform_pec_aim_base_l, loc=<0 0 0>, rot=<0 0 0 1>}
 deform_pec_base_l {parent=deform_pec_aim_l, loc=<16.3279 0 0>, rot=<0 0 0 1>}
 deform_pec_l {parent=deform_pec_base_l, loc=<0 0 0>, rot=<0 0 0 1>}
 deform_lat_l_base {parent=spine_03, loc=<2.5566 -10.3716 -17.6918>, rot=<0 0 0 1>}
 deform_lat_l {parent=deform_lat_l_base, loc=<0 0 0>, rot=<0 0 0 1>}
 deform_lat_r_base {parent=spine_03, loc=<2.5568 -10.3716 17.6918>, rot=<0 0 1 0>}
 deform_lat_r {parent=deform_lat_r_base, loc=<0 0 0>, rot=<0 0 0 -1>}
 thigh_l {parent=pelvis, loc=<-1.8716 -0.0109 -11.9936>, rot=<0.0012 0.0572 0.0155 -0.9982>}
 calf_l {parent=thigh_l, loc=<-41.9178 -0.2246 0.0926>, rot=<0.0006 -0.0161 -0.0391 -0.9991>}
 calf_twist_01_l {parent=calf_l, loc=<-32.9335 -0.0967 0.0397>, rot=<0.0237 0.0095 -0.0035 -0.9997>}
 calf_twist_02_l {parent=calf_l, loc=<-15.9268 -0.0967 0.0397>, rot=<0.0237 0.0095 -0.0035 -0.9997>}
 foot_l {parent=calf_l, loc=<-38.7057 1.2438 0.2452>, rot=<0.0975 -0.0474 0.0317 -0.9936>}
 ball_l {parent=foot_l, loc=<-6.1614 13.1469 0.6807>, rot=<0.0001 -0.0001 -0.7063 -0.7079>}
 ctrl_ball_end_l {parent=ball_l, loc=<6.025 1.6108 1.8202>, rot=<-0.999 0 -0.045 0.0001>}
 deform_knee_in_base_l {parent=calf_l, loc=<1.298 -8.0356 -0.0496>, rot=<0 0 0 1>}
 deform_knee_in_l {parent=deform_knee_in_base_l, loc=<0 0 0>, rot=<0 0 0 1>}
 deform_knee_out_base_l {parent=calf_l, loc=<0 6.8739 0>, rot=<0 0 0 1>}
 deform_knee_out_l {parent=deform_knee_out_base_l, loc=<0 0 0>, rot=<0 0 0 1>}
 reparent_foot_l {parent=calf_l, loc=<-42.2237 0.824 1.3998>, rot=<-0.5815 -0.8108 0.0672 0.0073>}
 reparent_ball_l {parent=reparent_foot_l, loc=<10.9798 0 -0.0001>, rot=<-0.0024 0.0006 -0.1969 -0.9804>}
 ball_end_l {parent=reparent_ball_l, loc=<9.946 0 0>, rot=<0 0 0 1>}
 dyn_reparent_foot_bk_base_l {parent=reparent_foot_l, loc=<-2.0438 -0.0999 -0.2245>, rot=<0.4213 0.5679 0.3561 -0.6109>}
 dyn_reparent_foot_bk_l {parent=dyn_reparent_foot_bk_base_l, loc=<0 0 0>, rot=<0 0 0 1>}
 thigh_twist_01_l {parent=thigh_l, loc=<-21.6303 0 0>, rot=<0.0013 0 0 -1>}
 deform_groin_base_l {parent=thigh_l, loc=<-2.1908 8.5661 -1.7442>, rot=<0 0 0 1>}
 deform_groin_l {parent=deform_groin_base_l, loc=<0 0 0>, rot=<0 0 0 1>}
 thigh_r {parent=pelvis, loc=<-1.8714 -0.0109 11.9936>, rot=<0.0572 -0.0012 0.9982 0.0155>}
 calf_r {parent=thigh_r, loc=<41.918 0.2246 -0.0926>, rot=<0.0006 -0.0161 -0.0391 -0.9991>}
 calf_twist_01_r {parent=calf_r, loc=<32.9336 0.0967 -0.0397>, rot=<0.0237 0.0095 -0.0035 -0.9997>}
 calf_twist_02_r {parent=calf_r, loc=<15.9268 0.0967 -0.0397>, rot=<0.0237 0.0095 -0.0035 -0.9997>}
 foot_r {parent=calf_r, loc=<38.7058 -1.2438 -0.2453>, rot=<0.0975 -0.0473 0.0317 -0.9936>}
 ball_r {parent=foot_r, loc=<6.1614 -13.1469 -0.6807>, rot=<0.0001 -0.0001 -0.7063 -0.7079>}
 ctrl_ball_end_r {parent=ball_r, loc=<-6.025 -1.6107 -1.8202>, rot=<-0.999 0 -0.045 0.0001>}
 deform_knee_in_base_r {parent=calf_r, loc=<-1.2981 8.0356 0.0496>, rot=<0 0 0 1>}
 deform_knee_in_r {parent=deform_knee_in_base_r, loc=<0 0 0>, rot=<0 0 0 1>}
 deform_knee_out_base_r {parent=calf_r, loc=<0 -6.8739 -0.0001>, rot=<0 0 0 1>}
 deform_knee_out_r {parent=deform_knee_out_base_r, loc=<0 0 0>, rot=<0 0 0 1>}
 reparent_foot_r {parent=calf_r, loc=<42.2237 -0.824 -1.3999>, rot=<-0.5815 -0.8109 0.0657 0.0058>}
 reparent_ball_r {parent=reparent_foot_r, loc=<-10.9798 0 0>, rot=<-0.0024 0.0006 -0.1969 -0.9804>}
 ball_end_r {parent=reparent_ball_r, loc=<-9.946 0 0>, rot=<0 0 0 1>}
 dyn_reparent_foot_bk_base_r {parent=reparent_foot_r, loc=<2.0438 0.0999 0.2243>, rot=<0.6109 -0.3561 0.5679 0.4213>}
 dyn_reparent_foot_bk_r {parent=dyn_reparent_foot_bk_base_r, loc=<0 0 0>, rot=<0 0 0 1>}
 thigh_twist_01_r {parent=thigh_r, loc=<21.6303 0 0>, rot=<0.0013 0 0 -1>}
 deform_groin_base_r {parent=thigh_r, loc=<2.191 -8.5661 1.7442>, rot=<0 0 0 1>}
 deform_groin_r {parent=deform_groin_base_r, loc=<0 0 0>, rot=<0 0 0 -1>}
 deform_glute_base_l {parent=pelvis, loc=<-10.1877 -11.6217 -12.899>, rot=<0.0275 0.0501 -0.4472 -0.8926>}
 deform_glute_l {parent=deform_glute_base_l, loc=<0 0 0>, rot=<0 0 0 1>}
 deform_glute_base_r {parent=pelvis, loc=<-10.1877 -11.6218 12.899>, rot=<0.0501 -0.0275 0.8926 -0.4472>}
 deform_glute_r {parent=deform_glute_base_r, loc=<0 0 0>, rot=<0 0 0 1>}
 Arm_L_PV {controlType=EULER_TRANSFORM, parent=root, type=Control, loc=<0 0 0>, rot=<0 0 0 1>}
 Arm_R_IK {controlType=EULER_TRANSFORM, parent=root, type=Control, loc=<0 0 0>, rot=<0 0 0 1>}
 Arm_L_IK {controlType=EULER_TRANSFORM, parent=root, type=Control, loc=<0 0 0>, rot=<0 0 0 1>}
 Arm_R_PV {controlType=EULER_TRANSFORM, parent=root, type=Control, loc=<0 0 0>, rot=<0 0 0 1>}
]
variables [
 isFrontend {type=bool, default=False}
 isGauntletsEquip {type=bool, default=True}
 isPlayingEmote {type=bool, default=False}
]
