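# Animation Blueprint for the DustyBun large male player body, adjusting a pose copied from the attached parent mesh.
# Additive head and neck bone modifiers feed a Control Rig driven by emote, gauntlet and frontend flags, followed by calf scaling.
# The update event detects gauntlets and boots by mesh-name substrings.
# It also computes counter-rotations for head and neck by negating and scaling each bone's reference-pose delta by X Factor.
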AnimBlueprint /BRCosmetics/Characters/Player/Male/Large/Bodies/M_LRG_DustyBun/Meshes/M_LRG_DustyBun_AnimBP
target_skeleton=&/M_LRG_DustyBun_Skeleton
graphs [
 AnimGraph {nodes [
   n1 = "Transform (Modify) Bone - Bone: head"!(ComponentPose=Local To Component(LocalPose=Copy Pose From Mesh() {bUseAttachedParent=true}), Alpha=1, Rotation="0,0,0") -> (Pose) {
    rotationMode=BMM_Additive
    rotationSpace=BCS_BoneSpace
    translationMode=BMM_Additive
    alphaBoolBlend {blendOption=Linear}
    boneToModify {boneName=head}
   }
   n2 = "Transform (Modify) Bone - Bone: neck_02"!(ComponentPose=@n1, Alpha=1, Rotation="0,0,0") -> (Pose) {
    rotationMode=BMM_Additive
    rotationSpace=BCS_BoneSpace
    translationMode=BMM_Additive
    alphaBoolBlend {blendOption=Linear}
    boneToModify {boneName=neck_02}
   }
   n3 = "Transform (Modify) Bone - Bone: neck_01"!(ComponentPose=@n2, Alpha=1, Rotation="0,0,0") -> (Pose) {
    rotationMode=BMM_Additive
    rotationSpace=BCS_BoneSpace
    translationMode=BMM_Additive
    alphaBoolBlend {blendOption=Linear}
    boneToModify {boneName=neck_01}
   }
   n4 = Control Rig!(
    Source=Component To Local(ComponentPose=@n3)
    Alpha
    isPlayingEmote:bool=Get bIsPlayingEmote()
    isGauntletsEquip:bool=Get GauntlesEquip() -> (GauntlesEquip:bool)
    isFrontend:bool=Get bIsInFrontEnd()
   ) -> (Pose) {
    alphaBoolBlend {blendOption=Linear}
    controlRigAssetReference {blueprintRigClass=&/M_LRG_DustyBun_CtrlRig.M_LRG_DustyBun_CtrlRig_C}
   }
   n5 = "Transform (Modify) Bone - Bone: calf_l"!(ComponentPose=Local To Component(LocalPose=@n4), bAlphaBoolEnabled:bool=True, Translation="0,0,0", Rotation="0,0,0", Scale="1,0.5,0.5") -> (Pose) {alphaInputType=Bool, scaleMode=BMM_Additive, alphaBoolBlend {blendOption=Linear}, boneToModify {boneName=calf_l}, scale {y=0.5, z=0.5}}
   n6 = "Transform (Modify) Bone - Bone: calf_r"!(ComponentPose=@n5, bAlphaBoolEnabled:bool=True, Translation="0,0,0", Rotation="0,0,0", Scale="1,0.5,0.5") -> (Pose) {alphaInputType=Bool, scaleMode=BMM_Additive, alphaBoolBlend {blendOption=Linear}, boneToModify {boneName=calf_r}, scale {y=0.5, z=0.5}}
   Output Pose(Result=Component To Local(ComponentPose=@n6)) {layerGroup=None}
  ]}
 EventGraph {nodes [
   Event BlueprintUpdateAnimation()
   > n1 = Sequence!(execute) -> (then_0, then_1, then_2:Exec)
   n2 = Get As Fort Player Pawn Athena() -> (As Fort Player Pawn Athena:Fort Player Pawn Athena Object Reference)
   > Is Valid(InputObject=@n2)
   n3 = GetDisplayName(Object=GetSkeletalMeshForPartType(self=@n2, PartType=Gameplay))
   > Set GauntlesEquip(GauntlesEquip:bool=OR Boolean!(
     A=Contains(SearchIn=@n3, Substring=GauntletsTemplate_Gadget)
     B=Contains(SearchIn=@n3, Substring=SprocketPoppy_Gauntlet)
     C:bool=Contains(SearchIn=@n3, Substring=CoffeePuff_Gauntlet)
    ) -> (ReturnValue)) -> (Output_Get:bool)
   n4 = Get As Fort Player Pawn Athena()
   Is Valid(exec=@n1.then_1, InputObject=@n4)
   n5 = GetSkeletalMeshForPartType(self=@n4, PartType=ExtraPart)
   > Set CustomBootsEquip(CustomBootsEquip:bool=Contains(SearchIn=GetDisplayName(Object=@n5), Substring=BootsTemplate_Gadget)) -> (Output_Get:bool)
   n6 = Get X Factor() -> (X Factor:double)
   n7 = Rotation adjustment!(
    execute=@n1.then_2
    self
    X Sign:double=-1
    X factor:double=@n6
    Y Sign:double=-1
    Y factor:double=@n6
    Z Sign:double=-1
    Z factor:double=@n6
    Bone To Limit:Name=head
    Reference Bone:Name=neck_02
   ) -> (then, Limited Rotation:Rotator)
   > Set headRotLimited(headRotLimited:Rotator=@n7) -> (Output_Get:Rotator)
   > n8 = Rotation adjustment(X Sign=-1, X factor=@n6, Y Sign=-1, Y factor=@n6, Z Sign=-1, Z factor=@n6, Bone To Limit=neck_02, Reference Bone=neck_01)
   > Set Neck_02 Rot Limited(Neck_02 Rot Limited:Rotator=@n8) -> (Output_Get:Rotator)
   > n9 = Rotation adjustment(X Sign=-1, X factor=@n6, Y Sign=-1, Y factor=@n6, Z Sign=-1, Z factor=@n6, Bone To Limit=neck_01, Reference Bone=spine_05)
   > Set Neck_01 Rot Limited(Neck_01 Rot Limited:Rotator=@n9) -> (Output_Get:Rotator)
   Event OnCharacterCustomizationLoaded()
   n10 = TryGetPawnOwner()
   > Is Valid(InputObject=@n10)
   > n11 = Cast To FortPlayerPawnAthena(Object=@n10)
   > Set As Fort Player Pawn Athena(As Fort Player Pawn Athena:Fort Player Pawn Athena Object Reference=@n11) -> (Output_Get:Fort Player Pawn Athena Object Reference)
  ]}
 Rotation adjustment {nodes [
   n1 = Rotation adjustment!() -> (
    then
    X Sign:double=~
    X factor:double=~
    Y Sign:double=~
    Y factor:double=~
    Z Sign:double=~
    Z factor:double=~
    Bone To Limit:Name
    Reference Bone:Name
   ) {is=title K2Node_FunctionEntry}
   n2 = GetOwnerSkeletalMeshComponent()
   n3 = BreakRotator(InRot=BreakTransform(InTransform=GetDeltaTransformFromRefPose(self=@n2, BoneName=@n1.Bone To Limit, BaseName=@n1.Reference Bone)).Rotation)
   > Return Node!(
    execute
    Limited Rotation:Rotator=MakeRotator(
     Roll=float * float!(A=@n3, B=@n1, C:double=@n1.X factor) -> (ReturnValue)
     Pitch=float * float(A=@n3.Pitch, B=@n1.Y Sign, C=@n1.Y factor)
     Yaw=float * float(A=@n3.Yaw, B=@n1.Z Sign, C=@n1.Z factor)
    )
   )
  ]}
]
variables [
 As Fort Player Pawn Athena {type=FortPlayerPawnAthena}
 CustomBootsEquip {type=boolean}
 Extra Part {type=AthenaCharacterPartItemDefinition, default=null}
 GauntlesEquip {type=boolean}
 headRotLimited {type=Rotator}
 Neck_01 Rot Limited {type=Rotator}
 Neck_02 Rot Limited {type=Rotator}
 X Factor {type=number, default=0.8}
]
