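# Animation Blueprint for the DustyBun large male face accessory, adding physics motion to left and right feather bones.
# The AnimGraph copies the attached parent's pose, passes it through additive head and neck nodes, springy Anim Dynamics feathers, then additive offsets on three feathers per side.
# The EventGraph computes scaled, negated head and neck deltas from reference pose, plus a speed-smoothed feather bend; no graph connection for these is shown.
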AnimBlueprint /BRCosmetics/Characters/Player/Male/Large/Bodies/M_LRG_DustyBun/Meshes/Parts/M_LRG_DustyBun_FaceAcc_AnimBP
target_skeleton=&/M_LRG_DustyBun_FaceAcc_Skeleton
graphs [
 AnimGraph {nodes [
   n1 = "Transform (Modify) Bone - Bone: head"!(ComponentPose=Local To Component(LocalPose=Copy Pose From Mesh() {bUseAttachedParent=true}), Alpha=1, Rotation="0,0,0") -> (Pose) {
    rotationMode=BMM_Additive
    rotationSpace=BCS_BoneSpace
    translationMode=BMM_Additive
    alphaBoolBlend {blendOption=Linear}
    boneToModify {boneName=head}
   }
   n2 = "Transform (Modify) Bone - Bone: neck_02"!(ComponentPose=@n1, Alpha=1, Rotation="0,0,0") -> (Pose) {
    rotationMode=BMM_Additive
    rotationSpace=BCS_BoneSpace
    translationMode=BMM_Additive
    alphaBoolBlend {blendOption=Linear}
    boneToModify {boneName=neck_02}
   }
   n3 = "Transform (Modify) Bone - Bone: neck_01"!(ComponentPose=@n2, Alpha=1, Rotation="0,0,0") -> (Pose) {
    rotationMode=BMM_Additive
    rotationSpace=BCS_BoneSpace
    translationMode=BMM_Additive
    alphaBoolBlend {blendOption=Linear}
    boneToModify {boneName=neck_01}
   }
   n4 = Anim Dynamics - Bone: dyn_feather_l_01(ComponentPose=@n3, Alpha=1, ExternalForce="0,0,0") {
    angularDampingOverride=1
    angularSpringConstant=1300
    bAngularSpring=true
    bOverrideAngularDamping=true
    bUseGravityOverride=true
    simulationSpace=RootRelative
    alphaBoolBlend {blendOption=Linear}
    boundBone {boneName=dyn_feather_l_01}
    componentAppliedLinearAccClamp=<5000 5000 5000>
    gravityOverride {z=-500}
    physicsBodyDefinitions [{
      boundBone {boneName=dyn_feather_l_01}
      boxExtents=<11 11 11>
      constraintSetup {angularLimitsMax=<5 10 10>, angularLimitsMin=<-5 -10 -10>, angularTarget=<1 0 0>}
      localJointOffset=<0.1 0 0>
     }]
   }
   n5 = Anim Dynamics - Bone: dyn_feather_l_02(ComponentPose=@n4, Alpha=1, ExternalForce="0,0,0") {
    angularDampingOverride=1
    angularSpringConstant=1300
    bAngularSpring=true
    bOverrideAngularDamping=true
    bUseGravityOverride=true
    simulationSpace=RootRelative
    alphaBoolBlend {blendOption=Linear}
    boundBone {boneName=dyn_feather_l_02}
    componentAppliedLinearAccClamp=<5000 5000 5000>
    gravityOverride {z=-500}
    physicsBodyDefinitions [{
      boundBone {boneName=dyn_feather_l_02}
      boxExtents=<11 11 11>
      constraintSetup {angularLimitsMax=<5 10 10>, angularLimitsMin=<-5 -10 -10>, angularTarget=<1 0 0>}
      localJointOffset=<0.1 0 0>
     }]
   }
   n6 = Anim Dynamics - Bone: dyn_feather_l_03(ComponentPose=@n5, Alpha=1, ExternalForce="0,0,0") {
    angularDampingOverride=1
    angularSpringConstant=1300
    bAngularSpring=true
    bOverrideAngularDamping=true
    bUseGravityOverride=true
    simulationSpace=RootRelative
    alphaBoolBlend {blendOption=Linear}
    boundBone {boneName=dyn_feather_l_03}
    componentAppliedLinearAccClamp=<5000 5000 5000>
    gravityOverride {z=-500}
    physicsBodyDefinitions [{
      boundBone {boneName=dyn_feather_l_03}
      boxExtents=<11 11 11>
      constraintSetup {angularLimitsMax=<5 10 10>, angularLimitsMin=<-5 -10 -10>, angularTarget=<1 0 0>}
      localJointOffset=<0.1 0 0>
     }]
   }
   n7 = Anim Dynamics - Bone: dyn_feather_l_04(ComponentPose=@n6, Alpha=1, ExternalForce="0,0,0") {
    angularDampingOverride=1
    angularSpringConstant=1300
    bAngularSpring=true
    bOverrideAngularDamping=true
    bUseGravityOverride=true
    simulationSpace=RootRelative
    alphaBoolBlend {blendOption=Linear}
    boundBone {boneName=dyn_feather_l_04}
    componentAppliedLinearAccClamp=<5000 5000 5000>
    gravityOverride {z=-500}
    physicsBodyDefinitions [{
      boundBone {boneName=dyn_feather_l_04}
      boxExtents=<11 11 11>
      constraintSetup {angularLimitsMax=<5 10 10>, angularLimitsMin=<-5 -10 -10>, angularTarget=<1 0 0>}
      localJointOffset=<0.1 0 0>
     }]
   }
   n8 = Anim Dynamics - Bone: dyn_feather_l_05(ComponentPose=@n7, Alpha=1, ExternalForce="0,0,0") {
    angularDampingOverride=1
    angularSpringConstant=1300
    bAngularSpring=true
    bOverrideAngularDamping=true
    bUseGravityOverride=true
    simulationSpace=RootRelative
    alphaBoolBlend {blendOption=Linear}
    boundBone {boneName=dyn_feather_l_05}
    componentAppliedLinearAccClamp=<5000 5000 5000>
    gravityOverride {z=-500}
    physicsBodyDefinitions [{
      boundBone {boneName=dyn_feather_l_05}
      boxExtents=<11 11 11>
      constraintSetup {angularLimitsMax=<5 10 10>, angularLimitsMin=<-5 -10 -10>, angularTarget=<1 0 0>}
      localJointOffset=<0.1 0 0>
     }]
   }
   n9 = Anim Dynamics - Bone: dyn_feather_l_06(ComponentPose=@n8, Alpha=1, ExternalForce="0,0,0") {
    angularDampingOverride=1
    angularSpringConstant=1300
    bAngularSpring=true
    bOverrideAngularDamping=true
    bUseGravityOverride=true
    simulationSpace=RootRelative
    alphaBoolBlend {blendOption=Linear}
    boundBone {boneName=dyn_feather_l_06}
    componentAppliedLinearAccClamp=<5000 5000 5000>
    gravityOverride {z=-500}
    physicsBodyDefinitions [{
      boundBone {boneName=dyn_feather_l_06}
      boxExtents=<11 11 11>
      constraintSetup {angularLimitsMax=<5 10 10>, angularLimitsMin=<-5 -10 -10>, angularTarget=<1 0 0>}
      localJointOffset=<0.1 0 0>
     }]
   }
   n10 = Anim Dynamics - Bone: dyn_feather_r_01(ComponentPose=@n9, Alpha=1, ExternalForce="0,0,0") {
    angularDampingOverride=1
    angularSpringConstant=1500
    bAngularSpring=true
    bOverrideAngularDamping=true
    bUseGravityOverride=true
    simulationSpace=RootRelative
    alphaBoolBlend {blendOption=Linear}
    boundBone {boneName=dyn_feather_r_01}
    componentAppliedLinearAccClamp=<5000 5000 5000>
    gravityOverride {z=-500}
    physicsBodyDefinitions [{
      boundBone {boneName=dyn_feather_r_01}
      boxExtents=<12 12 12>
      constraintSetup {angularLimitsMax=<5 10 10>, angularLimitsMin=<-5 -10 -10>, angularTarget=<1 0 0>}
      localJointOffset=<0.1 0 0>
     }]
   }
   n11 = Anim Dynamics - Bone: dyn_feather_r_02(ComponentPose=@n10, Alpha=1, ExternalForce="0,0,0") {
    angularDampingOverride=1
    angularSpringConstant=1500
    bAngularSpring=true
    bOverrideAngularDamping=true
    bUseGravityOverride=true
    simulationSpace=RootRelative
    alphaBoolBlend {blendOption=Linear}
    boundBone {boneName=dyn_feather_r_02}
    componentAppliedLinearAccClamp=<5000 5000 5000>
    gravityOverride {z=-500}
    physicsBodyDefinitions [{
      boundBone {boneName=dyn_feather_r_02}
      boxExtents=<12 12 12>
      constraintSetup {angularLimitsMax=<5 10 10>, angularLimitsMin=<-5 -10 -10>, angularTarget=<1 0 0>}
      localJointOffset=<0.1 0 0>
     }]
   }
   n12 = Anim Dynamics - Bone: dyn_feather_r_03(ComponentPose=@n11, Alpha=1, ExternalForce="0,0,0") {
    angularDampingOverride=1
    angularSpringConstant=1500
    bAngularSpring=true
    bOverrideAngularDamping=true
    bUseGravityOverride=true
    simulationSpace=RootRelative
    alphaBoolBlend {blendOption=Linear}
    boundBone {boneName=dyn_feather_r_03}
    componentAppliedLinearAccClamp=<5000 5000 5000>
    gravityOverride {z=-500}
    physicsBodyDefinitions [{
      boundBone {boneName=dyn_feather_r_03}
      boxExtents=<12 12 12>
      constraintSetup {angularLimitsMax=<5 10 10>, angularLimitsMin=<-5 -10 -10>, angularTarget=<1 0 0>}
      localJointOffset=<0.1 0 0>
     }]
   }
   n13 = Anim Dynamics - Bone: dyn_feather_r_04(ComponentPose=@n12, Alpha=1, ExternalForce="0,0,0") {
    angularDampingOverride=1
    angularSpringConstant=1500
    bAngularSpring=true
    bOverrideAngularDamping=true
    bUseGravityOverride=true
    simulationSpace=RootRelative
    alphaBoolBlend {blendOption=Linear}
    boundBone {boneName=dyn_feather_r_04}
    componentAppliedLinearAccClamp=<5000 5000 5000>
    gravityOverride {z=-500}
    physicsBodyDefinitions [{
      boundBone {boneName=dyn_feather_r_04}
      boxExtents=<12 12 12>
      constraintSetup {angularLimitsMax=<5 10 10>, angularLimitsMin=<-5 -10 -10>, angularTarget=<1 0 0>}
      localJointOffset=<0.1 0 0>
     }]
   }
   n14 = Anim Dynamics - Bone: dyn_feather_r_05(ComponentPose=@n13, Alpha=1, ExternalForce="0,0,0") {
    angularDampingOverride=1
    angularSpringConstant=1500
    bAngularSpring=true
    bOverrideAngularDamping=true
    bUseGravityOverride=true
    simulationSpace=RootRelative
    alphaBoolBlend {blendOption=Linear}
    boundBone {boneName=dyn_feather_r_05}
    componentAppliedLinearAccClamp=<5000 5000 5000>
    gravityOverride {z=-500}
    physicsBodyDefinitions [{
      boundBone {boneName=dyn_feather_r_05}
      boxExtents=<12 12 12>
      constraintSetup {angularLimitsMax=<5 10 10>, angularLimitsMin=<-5 -10 -10>, angularTarget=<1 0 0>}
      localJointOffset=<0.1 0 0>
     }]
   }
   n15 = Anim Dynamics - Bone: dyn_feather_r_06(ComponentPose=@n14, Alpha=1, ExternalForce="0,0,0") {
    angularDampingOverride=1
    angularSpringConstant=1500
    bAngularSpring=true
    bOverrideAngularDamping=true
    bUseGravityOverride=true
    simulationSpace=RootRelative
    alphaBoolBlend {blendOption=Linear}
    boundBone {boneName=dyn_feather_r_06}
    componentAppliedLinearAccClamp=<5000 5000 5000>
    gravityOverride {z=-500}
    physicsBodyDefinitions [{
      boundBone {boneName=dyn_feather_r_06}
      boxExtents=<12 12 12>
      constraintSetup {angularLimitsMax=<5 10 10>, angularLimitsMin=<-5 -10 -10>, angularTarget=<1 0 0>}
      localJointOffset=<0.1 0 0>
     }]
   }
   n16 = "Transform (Modify) Bone - Bone: dyn_feather_r_01"!(ComponentPose=@n15, Alpha=1, Rotation="0.000,12.661,-18.715") -> (Pose) {
    rotationMode=BMM_Additive
    translationSpace=BCS_BoneSpace
    alphaBoolBlend {blendOption=Linear}
    boneToModify {boneName=dyn_feather_r_01}
    rotation=<rot 0 12.661 -18.715>
   }
   n17 = "Transform (Modify) Bone - Bone: dyn_feather_r_02"!(ComponentPose=@n16, Alpha=1, Rotation="0.000,12.381,-15.468") -> (Pose) {
    rotationMode=BMM_Additive
    translationSpace=BCS_BoneSpace
    alphaBoolBlend {blendOption=Linear}
    boneToModify {boneName=dyn_feather_r_02}
    rotation {roll=-15.468, yaw=12.381}
   }
   n18 = "Transform (Modify) Bone - Bone: dyn_feather_r_03"!(ComponentPose=@n17, Alpha=1, Rotation="0,28.868,-26.581") -> (Pose) {
    rotationMode=BMM_Additive
    translationSpace=BCS_BoneSpace
    alphaBoolBlend {blendOption=Linear}
    boneToModify {boneName=dyn_feather_r_03}
    rotation {roll=-26.581, yaw=28.868}
   }
   n19 = "Transform (Modify) Bone - Bone: dyn_feather_l_01"!(ComponentPose=@n18, Alpha=1, Rotation="-0.000,-17.362,-25.612") -> (Pose) {
    rotationMode=BMM_Additive
    translationSpace=BCS_BoneSpace
    alphaBoolBlend {blendOption=Linear}
    boneToModify {boneName=dyn_feather_l_01}
    rotation=<rot 0 -17.362 -25.612>
   }
   n20 = "Transform (Modify) Bone - Bone: dyn_feather_l_02"!(ComponentPose=@n19, Alpha=1, Rotation="0.000,-14.191,-25.663") -> (Pose) {
    rotationMode=BMM_Additive
    translationSpace=BCS_BoneSpace
    alphaBoolBlend {blendOption=Linear}
    boneToModify {boneName=dyn_feather_l_02}
    rotation=<rot 0 -14.191 -25.663>
   }
   n21 = "Transform (Modify) Bone - Bone: dyn_feather_l_03"!(ComponentPose=@n20, Alpha=1, Rotation="0.000,-15.522,-19.023") -> (Pose) {
    rotationMode=BMM_Additive
    translationSpace=BCS_BoneSpace
    alphaBoolBlend {blendOption=Linear}
    boneToModify {boneName=dyn_feather_l_03}
    rotation=<rot 0 -15.522 -19.023>
   }
   Output Pose(Result=Component To Local(ComponentPose=@n21)) {layerGroup=None}
  ]}
 EventGraph {nodes [
   n1 = Event BlueprintUpdateAnimation()
   > n2 = Sequence()
   n3 = Get X Factor() -> (X Factor:double)
   > n4 = Rotation adjustment!(
    execute
    self
    X Sign:double=-1
    X factor:double=@n3
    Y Sign:double=-1
    Y factor:double=@n3
    Z Sign:double=-1
    Z factor:double=@n3
    Bone To Limit:Name=head
    Reference Bone:Name=neck_02
   ) -> (then, Limited Rotation:Rotator)
   > Set headRotLimited(headRotLimited:Rotator=@n4) -> (Output_Get:Rotator)
   > n5 = Rotation adjustment(X Sign=-1, X factor=@n3, Y Sign=-1, Y factor=@n3, Z Sign=-1, Z factor=@n3, Bone To Limit=neck_02, Reference Bone=neck_01)
   > Set Neck_02 Rot Limited(Neck_02 Rot Limited:Rotator=@n5) -> (Output_Get:Rotator)
   > n6 = Rotation adjustment(X Sign=-1, X factor=@n3, Y Sign=-1, Y factor=@n3, Z Sign=-1, Z factor=@n3, Bone To Limit=neck_01, Reference Bone=spine_05)
   > Set Neck_01 Rot Limited(Neck_01 Rot Limited:Rotator=@n6) -> (Output_Get:Rotator)
   n7 = Vector Length(A=Get PawnVelocity())
   Set feathers bend(
    execute=@n2.then_1
    feathers bend:double=FInterpTo(
     Current=Get feathers bend() -> (feathers bend:double)
     Target=Remap Float(Input=@n7, OldRange=(X=550,Y=780), NewRange=(X=0,Y=1), Clamp=true) -> (Output=~)
     DeltaTime=@n1.DeltaTimeX
     InterpSpeed=5
    )
   ) -> (Output_Get:double)
  ]}
 Rotation adjustment {nodes [
   n1 = Rotation adjustment!() -> (
    then
    X Sign:double=~
    X factor:double=~
    Y Sign:double=~
    Y factor:double=~
    Z Sign:double=~
    Z factor:double=~
    Bone To Limit:Name
    Reference Bone:Name
   ) {is=title K2Node_FunctionEntry}
   n2 = GetOwnerSkeletalMeshComponent()
   n3 = BreakRotator(InRot=BreakTransform(InTransform=GetDeltaTransformFromRefPose(self=@n2, BoneName=@n1.Bone To Limit, BaseName=@n1.Reference Bone)).Rotation)
   > Return Node!(
    execute
    Limited Rotation:Rotator=MakeRotator(
     Roll=float * float!(A=@n3, B=@n1, C:double=@n1.X factor) -> (ReturnValue)
     Pitch=float * float(A=@n3.Pitch, B=@n1.Y Sign, C=@n1.Y factor)
     Yaw=float * float(A=@n3.Yaw, B=@n1.Z Sign, C=@n1.Z factor)
    )
   )
  ]}
]
variables [
 feathers bend {type=number}
 headRotLimited {type=Rotator}
 Neck_01 Rot Limited {type=Rotator}
 Neck_02 Rot Limited {type=Rotator}
 X Factor {type=number, default=0.8}
]
